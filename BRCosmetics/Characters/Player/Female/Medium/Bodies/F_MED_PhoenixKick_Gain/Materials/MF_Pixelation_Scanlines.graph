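# Screen-aligned material function for pixelation and scanlines on monitor or hologram looks, stored with the F_MED_PhoenixKick_Gain character body.
# Camera-to-object view depth sets a whole-number, aspect-corrected line density for screen-UV sines.
# Phase-shifted RGB sines add colour fringing, scaled by Bloom1/Bloom2, clamped to at least one, then oversaturated via negative Desaturation.
# A Secondary Lines output adds two time-swept, sizeable bands; FOV compensation and jitter removal nodes are unconnected.
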MaterialFunction /BRCosmetics/Characters/Player/Female/Medium/Bodies/F_MED_PhoenixKick_Gain/Materials/MF_Pixelation_Scanlines
graph {
 properties {description="Screen aligned scanlines, useful for monitors and holograms", libraryCategoriesText [VFX]}
 nodes [
  MaterialExpressionSubtract_0 = Subtract(A=CameraPositionWS(), B=ObjectPositionWS())
  n1 = Custom!(x=@MaterialExpressionSubtract_0) {
   code="length(x)"
   inputs [{
     inputName=x
     input {expression=&self:MaterialExpressionSubtract_0}
    }]
  }
  n2 = Constant3Vector() {constant=<rgba 0 0 -1 0>}
  n3 = Multiply(
   A=@n1
   B=DotProduct(A=Normalize(VectorInput=@MaterialExpressionSubtract_0), B=Transform(None=@n2) {transformSourceType=TRANSFORMSOURCE_View})
  )
  Add(A=@n3)
  Constant() {r=20}
  n4 = ScreenPosition()
  n5 = Constant() {r=300}
  n6 = Divide(A=Subtract(A=@n3, B=@n5), B=Constant() {r=2000})
  n7 = LinearInterpolate(A=Constant() {r=150}, B=LinearInterpolate(A=Constant() {r=300}, B=Constant() {r=1000}, Alpha=@n6), Alpha=Saturate(None=Divide(A=@n3, B=@n5)))
  n8 = Subtract(A=@n7, B=Frac(None=@n7))
  n9 = ViewSize()
  n10 = Divide(A=ComponentMask(None=@n9) {r=true}, B=ComponentMask(None=@n9) {g=true})
  n11 = Multiply(A=MaterialFunctionCall(X=Multiply(A=@n8, B=@n10), Y=@n8) {materialFunction=&MakeFloat2}.Result) {constB=0.75}
  n12 = Constant() {r=0.03}
  n13 = Sine(None=Add(A=Multiply(A=@n4.ViewportUV, B=@n11) {constB=100}, B=@n12))
  n14 = Power(Base=ComponentMask(None=@n13) {r=true}, Exp=Constant() {r=0.1})
  n15 = Constant() {r=1}
  n16 = Sine(None=Subtract(A=Multiply(A=@n4.ViewportUV, B=@n11) {constB=100}, B=@n12))
  n17 = Sine(None=Multiply(A=@n4.ViewportUV, B=@n11) {constB=100})
  n18 = MaterialFunctionCall(
   R=ConstantBiasScale(None=ComponentMask(None=@n16) {g=true})
   G=ConstantBiasScale(None=ComponentMask(None=@n17) {g=true})
   B=ConstantBiasScale(None=ComponentMask(None=@n13) {g=true})
  ) {materialFunction=&AppendMany}
  n19 = Constant() {r=0.3}
  n20 = Add(
   A=Multiply(A=ScalarParameter() {defaultValue=0.1, parameterName=Bloom1}, B=Add(A=@n14, B=@n15))
   B=Multiply(A=Power(Base=@n18.RGB, Exp=@n19) {constExponent=0.3}, B=ScalarParameter() {defaultValue=1.5, parameterName=Bloom2}) {constB=2}
  )
  n21 = Constant() {r=1}
  FunctionOutput(None=Desaturation(None=Max(A=@n20, B=@n21), Fraction=Constant() {r=-10}))
  n22 = MaterialFunctionCall(World Position=WorldPosition().XYZ) {materialFunction=&TransformToClipSpace}
  n23 = Time()
  n24 = Subtract(A=@n22.Screen aligned 0-1 UV, B=ConstantBiasScale(None=Sine(None=@n23) {period=5.41}) {bias=1.3, scale=0.35})
  n25 = FunctionInput() {bUsePreviewValueAsDefault=true, inputName=Secondary Line A Size, inputType=FunctionInput_Scalar, previewValue=<96 0 0 1>}
  n26 = ComponentMask(None=Saturate(None=Multiply(A=@n24, B=@n25))) {g=true}
  n27 = Multiply(A=Multiply(A=@n26, B=OneMinus(None=@n26))) {constB=2.1}
  n28 = Subtract(A=@n22.Screen aligned 0-1 UV, B=ConstantBiasScale(None=Sine(None=@n23) {period=7.89}) {bias=1.25, scale=0.3})
  n29 = FunctionInput() {bUsePreviewValueAsDefault=true, inputName=Secondary Line B Size, inputType=FunctionInput_Scalar, previewValue=<84 0 0 1>}
  n30 = ComponentMask(None=Saturate(None=Multiply(A=@n28, B=@n29))) {g=true}
  n31 = Multiply(A=Multiply(A=@n30, B=OneMinus(None=@n30))) {constB=2.2}
  FunctionOutput(None=Add(
    A=MaterialFunctionCall!(In=@n27) {materialFunction=/Engine/Functions/Engine_MaterialFunctions02/Utility/Round}.Result
    B=MaterialFunctionCall(In=@n31) {materialFunction=&Round}.Result
   )) {outputName=Secondary Lines, sortPriority=1}
  MaterialFunctionCall() {materialFunction=&ObjectPivotPoint}
  n32 = MaterialFunctionCall() {materialFunction=/Engine/Functions/Engine_MaterialFunctions02/Utility/FOV}
  Multiply(A=@n6, B=Saturate(None=InverseLinearInterpolate(Value=@n32.FOV angle in degrees) {constA=45, constB=90}) {desc=FOV compensation})
  Subtract(A=@n4.ViewportUV, B=Divide(A=ViewProperty() {property=MEVP_TemporalSampleOffset}.Property, B=ViewSize()))
 ]
}
