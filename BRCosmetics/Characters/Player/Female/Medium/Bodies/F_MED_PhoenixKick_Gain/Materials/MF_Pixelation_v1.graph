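# Material function for the PhoenixKick_Gain character body that gives it a toggleable pixelated, scanlined look, blended via the Pixelated parameter.
# It outputs grid-snapped UVs and a transformed normal, with explicit UV derivatives on texture samples for mip selection.
# Posterized diffuse, emissive, banded shadows and sky-light specular are multiplied by MF_Pixelation_Scanlines into Emissive, with BaseColor zeroed.
# A channel-masked option flattens the light vector's Z for shading.
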MaterialFunction /BRCosmetics/Characters/Player/Female/Medium/Bodies/F_MED_PhoenixKick_Gain/Materials/MF_Pixelation_v1
graph {nodes [
  n1 = ScalarParameter() {
   controlType=Enumeration
   enumeration=/Game/Global/Materials/Enums/E_Mat_OnOff
   group=Pixelation - General
   parameterName=Pixelated
   sortPriority=-1
  }
  n2 = MaterialFunctionCall!(
   Secondary Line A Size
   Secondary Line B Size
   Scanline Speed
   Number of Scanlines=ScalarParameter() {defaultValue=150, group=Pixelation - Scanlines, parameterName=Pixelation.Scanlines_Amount, sortPriority=-1}
   "Scanlines Thickness (Bias)"
   Scanlines Dark Value=ScalarParameter() {defaultValue=0.8, group=Pixelation - Scanlines, parameterName=Pixelation.Scanlines_MinValue, sortPriority=-1}
  ) {materialFunction=&/MF_Pixelation_Scanlines}
  n3 = Multiply(A=@n2.Secondary Lines) {constB=0.5}
  n4 = Add(
   A=DotProduct(A=NamedRerouteUsage(), B=NamedRerouteUsage())
   B=ScalarParameter() {defaultValue=0.5, group=Pixelation - Shading, parameterName=Pixelation.Shadow_Bias, sortPriority=0}
  )
  n5 = ScalarParameter() {defaultValue=1, group=Pixelation - Shading, parameterName=Pixelation.Shadow_Curve, sortPriority=1}
  n6 = ScalarParameter() {defaultValue=3, group=Pixelation - Shading, parameterName=Pixelation.Shadow_BandCount, sortPriority=2}
  n7 = Floor(None=Multiply(A=Power(Base=@n4, Exp=@n5), B=@n6))
  n8 = Saturate(None=Max(
    A=Divide(A=@n7, B=@n6)
    B=ScalarParameter() {defaultValue=0.2572, group=Pixelation - Shading, parameterName=Pixelation.Shadow_MinValue, sortPriority=3}
   ))
  n9 = TextureCoordinate()
  n10 = TextureSampleParameter2D!(UVs=NamedRerouteUsage(), "DDX(UVs)"=DDX(Value=@n9), "DDY(UVs)"=DDY(Value=@n9), Apply View MipBias) {
   group=Pixelation - General
   mipValueMode=TMVM_Derivative
   parameterName=Pixelation.Diffuse
   samplerSource=SSM_Wrap_WorldGroupSettings
   sortPriority=0
   texture=/EpicBaseTextures/Blanks/T_Default_BaseColor_White
  }
  n11 = TextureCoordinate()
  n12 = TextureSampleParameter2D(UVs=NamedRerouteUsage(), "DDX(UVs)"=DDX(Value=@n11), "DDY(UVs)"=DDY(Value=@n11)) {
   desc=Inherited from Character Parent. Don't rename or regroup
   group=00_Texture Inputs
   mipValueMode=TMVM_Derivative
   parameterName=Diffuse
   samplerSource=SSM_Wrap_WorldGroupSettings
   sortPriority=0
   texture=&T_Default_BaseColor_White
  }
  n13 = ScalarParameter() {defaultValue=32, group=Pixelation - General, parameterName=Pixelation.BaseColor_Steps, sortPriority=11}
  n14 = Multiply(
   A=Multiply(
    A=StaticSwitchParameter(True=@n10.RGB, False=@n12.RGB) {group=Pixelation - General, parameterName=Pixelation.UseCustomDiffuse, sortPriority=0}
    B=ScalarParameter() {defaultValue=1.25, group=Pixelation - General, parameterName=Pixelation.BaseColor_Intensity, sortPriority=10}
   )
   B=@n13
  )
  n15 = TextureSampleParameter2D!(UVs=NamedRerouteUsage(), Apply View MipBias) {
   desc=Inherited from Character Parent. Don't rename or regroup
   group=06_Emissive
   parameterName=Emissive
   texture=/EpicBaseTextures/Blanks/T_Default_EmissiveColor_Black
  }
  n16 = ScalarParameter() {defaultValue=2, desc=Inherited from Character Parent. Don't rename or regroup, group=06_Emissive, parameterName=EmissiveBrightness}
  n17 = Max(A=Divide(A=Floor(None=@n14), B=@n13), B=Multiply(A=@n15.RGB, B=@n16))
  n18 = LinearInterpolate(
   A=Multiply(A=@n17, B=NamedRerouteDeclaration(None=@n8) {name=Shadow dither, nodeColor=<rgba 1 0.6562 0 1>})
   B=@n17
   Alpha=NamedRerouteUsage()
  )
  MaterialExpressionCameraVectorWS_SP00 = CameraVectorWS()
  MaterialExpressionAdd_SP00 = Add(A=NamedRerouteUsage(), B=@MaterialExpressionCameraVectorWS_SP00)
  MaterialExpressionNormalize_SP00 = Normalize(VectorInput=@MaterialExpressionAdd_SP00)
  MaterialExpressionDotProduct_SP00 = DotProduct(A=@MaterialExpressionNormalize_SP00, B=NamedRerouteUsage())
  MaterialExpressionScalarParameter_SP00 = ScalarParameter() {defaultValue=6, group=Pixelation - Shading, parameterName=Pixelation.Specular_Sharpness, sortPriority=4}
  MaterialExpressionPower_SP00 = Power(Base=Saturate(None=@MaterialExpressionDotProduct_SP00), Exp=@MaterialExpressionScalarParameter_SP00)
  n19 = ScalarParameter() {defaultValue=3, group=Pixelation - Shading, parameterName=Pixelation.Shadow_BandCount, sortPriority=22}
  n20 = Divide(A=Floor(None=Multiply(A=@MaterialExpressionPower_SP00, B=@n19)), B=@n19)
  n21 = Multiply(A=SkyAtmosphereLightIlluminance!(World Position), B=@n20)
  n22 = Constant() {r=1}
  n23 = NamedRerouteUsage()
  n24 = Multiply(
   A=@n20
   B=Multiply(
    A=ScalarParameter() {defaultValue=0.05, group=Pixelation - Shading, parameterName=Pixelation.Specular_Intensity, sortPriority=5}
    B=Step(Y=@n22, X=@n23)
   )
  )
  n25 = Constant()
  n26 = Constant()
  n27 = SetMaterialAttributes!(
   MaterialAttributes
   EmissiveColor=Multiply(A=Add(A=@n2.Result, B=@n3), B=LinearInterpolate(A=@n18, B=@n21, Alpha=@n24))
   Normal=Constant3Vector() {constant=<rgba 0.5 0.5 1 0>}
   Specular=@n25
   Roughness=Constant() {r=1}
   BaseColor=@n25
   SubsurfaceColor=@n26
   Opacity=@n26
  ) {attributeSetTypes [
    #1
    #2
    #3
    #4
    #5
    #6
    #7
   ]}
  FunctionOutput(None=BlendMaterialAttributes(A=NamedRerouteUsage(), B=@n27, Alpha=@n1))
  n28 = TextureCoordinate()
  n29 = ScalarParameter() {defaultValue=250, group=Pixelation - General, parameterName=Pixelation.QuantizedScale, sortPriority=1}
  n30 = Add(A=Floor(None=Multiply(A=@n28, B=@n29))) {constB=0.5}
  FunctionOutput(None=NamedRerouteDeclaration(None=Divide(A=@n30, B=@n29)) {name=PixelatedUVs, nodeColor=<rgba 1 0.6797 0 1>}) {outputName=Pixelated UVs, sortPriority=1}
  n31 = TextureCoordinate()
  n32 = TextureSampleParameter2D!(UVs=@n31, "DDX(UVs)"=DDX(Value=@n31), "DDY(UVs)"=DDY(Value=@n31), Apply View MipBias) {
   group=Pixelation - General
   mipValueMode=TMVM_Derivative
   parameterName=Pixelation.Normal
   samplerType=SAMPLERTYPE_Normal
   sortPriority=0
   texture=&FlatNormalBlue
  }
  FunctionOutput(None=NamedRerouteDeclaration(None=Transform(None=@n32.RGB)) {name=PixelatedNormal, nodeColor=<rgba 0 1 1 1>}) {outputName=Pixelated Normal}
  n33 = GetMaterialAttributes(None=FunctionInput() {inputName=Material, inputType=FunctionInput_MaterialAttributes}) {attributeGetTypes [#5, #1]}
  NamedRerouteDeclaration(None=SetMaterialAttributes!(MaterialAttributes=@n33.MaterialAttributes, BaseColor=@n33.MaterialAttributes, EmissiveColor=@n33.MaterialAttributes) {attributeSetTypes [#5, #1]}) {name=IN_Mat, nodeColor=<rgba 1 0 0.6797 1>}
  NamedRerouteDeclaration(None=Max(A=Max(A=@n15.R, B=@n15.R), B=@n15.B)) {name=Emissive Mask, nodeColor=<rgba 1 0.3516 0 1>}
  n34 = SkyAtmosphereLightDirection()
  n35 = MaterialFunctionCall(Float3=@n34) {materialFunction=&BreakOutFloat3Components}
  n36 = ScalarParameter() {
   desc="0.0 = the light vector is locked on the XY plane.\r\n1.0 = the actual light vector is used  \r\n(for Faces use a blend value of 0,0)"
   group=Pixelation - Shading
   parameterName=Pixelation.LockedZLightVector_Blend
   sliderMax=1
   sortPriority=33
  }
  n37 = MaterialFunctionCall(X=@n35.R, Y=@n35.R, Z=Multiply(A=@n35.B, B=@n36)) {materialFunction=&MakeFloat3}
  n38 = ChannelMaskParameter(None=FunctionInput() {bUsePreviewValueAsDefault=true, inputName=TechArt Mask, previewValue=<1 1 1 1>}) {group=Pixelation - Shading, maskChannel=Green, parameterName=Pixelation.LockedZLightVector_Channel, defaultValue=<rgba 0 1 0 0>}
  n39 = ScalarParameter() {
   controlType=Enumeration
   enumeration=&E_Mat_OnOff
   group=Pixelation - Shading
   parameterName=Pixelation.UseLockedZLightVector
   sortPriority=30
  }
  NamedRerouteDeclaration(None=LinearInterpolate(A=@n34, B=Normalize(VectorInput=@n37.Result), Alpha=Multiply(A=@n38.R, B=@n39))) {name=LightDirection, nodeColor=<rgba 0 1 0.5781 1>}
 ]}
