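# Material function that adds stylized rim light and highlights to incoming material attributes, toggled by an Enable Rim Light parameter.
# View-space light directions, transformed and dotted with normals, drive a Fresnel-shaped rim and a power-sharpened highlight masked by a Tech Art Mask channel.
# Max merges these with plain Fresnel, and camera depth fade shifts near/far settings.
# The result blends toward attributes with rim color added to Emissive and scaled into Base Color.
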MaterialFunction /Game/MF_RimLight
graph {
 properties {bExposeToLibrary=true}
 nodes [
  n1 = GetMaterialAttributes(None=NamedRerouteUsage()) {attributeGetTypes [#1, #2]}
  n2 = VectorParameter() {group="CreamDecoy - Rimlights && Highlights", parameterName=RimLightColor, defaultValue=<rgba 0.1379 0.1735 0.3021 1>}
  n3 = Add(A=@n1.EmissiveColor, B=@n2.RGB)
  n4 = Multiply(A=@n3, B=ScalarParameter() {group="CreamDecoy - Rimlights && Highlights", parameterName=BaseColorContribution})
  n5 = ScalarParameter() {group="CreamDecoy - Rimlights && Highlights", parameterName=RimLight - Min}
  n6 = ScalarParameter() {defaultValue=1, group="CreamDecoy - Rimlights && Highlights", parameterName=RimLight - Max}
  n7 = VectorParameter() {group="CreamDecoy - Rimlights && Highlights", parameterName=RimLightScreenSpacePosition, defaultValue=<rgba 1 0.5 1 1>}
  n8 = Transform(None=@n7.RGB) {transformSourceType=TRANSFORMSOURCE_View}
  n9 = PixelNormalWS()
  n10 = Fresnel(ExponentIn=LinearInterpolate(
    A=ScalarParameter() {defaultValue=1, group="CreamDecoy - Rimlights && Highlights", parameterName=FresnelBaseExponent - Near}
    B=ScalarParameter() {defaultValue=1, group="CreamDecoy - Rimlights && Highlights", parameterName=FresnelBaseExponent - Far}
    Alpha=NamedRerouteUsage()
   )) {baseReflectFraction=0, exponent=1}
  n11 = Multiply(A=Saturate(None=DotProduct(A=@n8, B=@n9)), B=@n10)
  n12 = Power(Base=@n10, Exp=ScalarParameter() {defaultValue=4, group="CreamDecoy - Rimlights && Highlights", parameterName=FresnelPow})
  n13 = ScalarParameter() {defaultValue=0.25, group="CreamDecoy - Rimlights && Highlights", parameterName=FresnelIntensit}
  n14 = VectorParameter() {group="CreamDecoy - Rimlights && Highlights|Highlights", parameterName=HighlightsScreenSpacePosition, defaultValue=<rgba -0.25 0 -1 1>}
  n15 = Transform(None=@n14.RGB) {transformSourceType=TRANSFORMSOURCE_View}
  n16 = LinearInterpolate(A=PixelNormalWS(), B=VertexNormalWS())
  n17 = Power(
   Base=Saturate(None=DotProduct(A=@n15, B=@n16))
   Exp=ScalarParameter() {defaultValue=32, group="CreamDecoy - Rimlights && Highlights|Highlights", parameterName=HighlightExp}
  ) {constExponent=64}
  n18 = NamedRerouteUsage()
  n19 = ScalarParameter() {defaultValue=0.5, group="CreamDecoy - Rimlights && Highlights|Highlights", parameterName=HighligtsIntensity_Near}
  n20 = ScalarParameter() {defaultValue=0.1, group="CreamDecoy - Rimlights && Highlights|Highlights", parameterName=HighligtsIntensity_Far}
  n21 = NamedRerouteUsage()
  n22 = Max(
   A=Max(A=SmoothStep(Min=@n5, Max=@n6, Value=@n11), B=Multiply(A=@n12, B=@n13))
   B=Multiply(A=Multiply(A=@n17, B=@n18), B=LinearInterpolate(A=@n19, B=@n20, Alpha=@n21))
  )
  n23 = LinearInterpolate(
   A=Multiply(A=@n22, B=ScalarParameter() {defaultValue=1, group="CreamDecoy - Rimlights && Highlights", parameterName=RimLightColorIntensity - Near})
   B=Multiply(A=@n22, B=ScalarParameter() {defaultValue=1, group="CreamDecoy - Rimlights && Highlights", parameterName=RimLightColorIntensity - Far})
   Alpha=NamedRerouteUsage()
  )
  n24 = ScalarParameter() {
   controlType=Enumeration
   enumeration=/Game/Global/Materials/Enums/E_Mat_OnOff
   group="CreamDecoy - Rimlights && Highlights|Highlights"
   parameterName=Enable Rim Light
  }
  FunctionOutput(None=BlendMaterialAttributes(
    A=NamedRerouteUsage()
    B=SetMaterialAttributes!(MaterialAttributes=@n1.MaterialAttributes, BaseColor=@n4, EmissiveColor=@n3) {attributeSetTypes [#1, #2]}
    Alpha=Multiply(A=@n23, B=@n24)
   ))
  NamedRerouteDeclaration(None=FunctionInput() {inputName=Result, inputType=FunctionInput_MaterialAttributes}) {name=Material In, nodeColor=<rgba 1 0 0.8906 1>}
  NamedRerouteDeclaration(None=FunctionInput() {inputName=Tech Art Mask, sortPriority=1}) {name=FXMask In, nodeColor=<rgba 1 0 0.5859 1>}
  n25 = ChannelMaskParameter(None=NamedRerouteUsage()) {group="CreamDecoy - Rimlights && Highlights", maskChannel=Blue, parameterName=CheckersHighlightMask, defaultValue=<rgba 0 0 1 0>}
  NamedRerouteDeclaration(None=@n25.R) {name=HighlightsMask, nodeColor=<rgba 1 0.2109 0 1>}
  NamedRerouteDeclaration(None=MaterialFunctionCall(
    Fade Length=ScalarParameter() {defaultValue=512, group="CreamDecoy - Rimlights && Highlights", parameterName="Fade Length (S)"}
    Fade Offset=ScalarParameter() {defaultValue=350, group="CreamDecoy - Rimlights && Highlights", parameterName="Fade Offset (S)"}
   ) {materialFunction=&CameraDepthFade}.Result) {name=CameraDistanceFade, nodeColor=<rgba 0.5234 1 0 1>}
 ]
}
