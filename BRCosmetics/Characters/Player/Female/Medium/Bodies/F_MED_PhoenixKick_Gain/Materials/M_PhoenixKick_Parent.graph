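# Parent material for the PhoenixKick Gain character body, giving a retro, quantized look to the standard character parent with optional hair functions.
# Diffuse is sampled on snapped UVs with original derivatives, posterized, then combined with emissive, banded sky-light shading and stepped specular.
# The lit result goes to Emissive, with Base Color and Specular zeroed.
# World position offset snaps vertices to a camera-depth-scaled grid for vertex swimming; pixelation and dither helpers are also present.
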Material /BRCosmetics/Characters/Player/Female/Medium/Bodies/F_MED_PhoenixKick_Gain/Materials/M_PhoenixKick_Parent
graph {
 expressions {
  MaterialExpressionAbs_0 {class=MaterialExpressionAbs, inputs [None {from=MaterialExpressionDDX_0, output=""}], properties {}}
  MaterialExpressionAbs_1 {class=MaterialExpressionAbs, inputs [None {from=MaterialExpressionDDY_0, output=""}], properties {}}
  MaterialExpressionAbs_2 {class=MaterialExpressionAbs, inputs [None {from="", output=""}], properties {}}
  MaterialExpressionAdd_0 {
   class=MaterialExpressionAdd
   inputs [
    A {from=MaterialExpressionNamedRerouteUsage_5}
    B {from=MaterialExpressionPower_SP00}
   ]
  }
  MaterialExpressionAdd_1 {
   class=MaterialExpressionAdd
   inputs [
    A {from=MaterialExpressionMin_0}
    B {from=MaterialExpressionScalarParameter_22}
   ]
  }
  MaterialExpressionAdd_10 {
   class=MaterialExpressionAdd
   inputs [
    A {from=MaterialExpressionDotProduct_5}
    B {from=MaterialExpressionScalarParameter_32}
   ]
  }
  MaterialExpressionAdd_12 {
   class=MaterialExpressionAdd
   inputs [
    A {from=""}
    B {from=MaterialExpressionNamedRerouteDeclaration_4}
   ]
  }
  MaterialExpressionAdd_3 {
   class=MaterialExpressionAdd
   inputs [
    A {from=MaterialExpressionAbs_0}
    B {from=MaterialExpressionAbs_1}
   ]
  }
  MaterialExpressionAdd_4 {
   class=MaterialExpressionAdd
   inputs [
    A {from=""}
    B {from=MaterialExpressionNamedRerouteUsage_1}
   ]
  }
  MaterialExpressionAdd_5 {
   class=MaterialExpressionAdd
   inputs [
    A {from=MaterialExpressionObjectPositionWS_0}
    B {from=MaterialExpressionConstant3Vector_6}
   ]
  }
  MaterialExpressionAdd_6 {
   class=MaterialExpressionAdd
   inputs [
    A {from=MaterialExpressionPower_4}
    B {from=MaterialExpressionNamedRerouteUsage_5}
   ]
  }
  MaterialExpressionAdd_SP00 {
   class=MaterialExpressionAdd
   inputs [
    A {from=MaterialExpressionSkyAtmosphereLightDirection_SP00}
    B {from=MaterialExpressionCameraVectorWS_SP00}
   ]
  }
  MaterialExpressionArccosine_0 {class=MaterialExpressionArccosine, inputs [None {from=MaterialExpressionDotProduct_4, output=""}], properties {}}
  MaterialExpressionAtmosphericLightVector_2 {class=MaterialExpressionAtmosphericLightVector, inputs [], properties {}}
  MaterialExpressionAtmosphericLightVector_3 {class=MaterialExpressionAtmosphericLightVector, inputs [], properties {}}
  MaterialExpressionAtmosphericLightVector_4 {class=MaterialExpressionAtmosphericLightVector, inputs [], properties {}}
  MaterialExpressionBlendMaterialAttributes_0 {
   class=MaterialExpressionBlendMaterialAttributes
   inputs [
    A {from=MaterialExpressionNamedRerouteUsage_2, output=""}
    B {from=MaterialExpressionSetMaterialAttributes_0, output=""}
    Alpha {from="", output=""}
   ]
   properties {}
  }
  MaterialExpressionBlendMaterialAttributes_1 {
   class=MaterialExpressionBlendMaterialAttributes
   inputs [
    A {from=MaterialExpressionReroute_0, output=""}
    B {from=MaterialExpressionStaticSwitchParameter_60, output=""}
    Alpha {from=MaterialExpressionMaterialFunctionCall_5, output=HairMask}
   ]
   properties {}
  }
  MaterialExpressionCameraVectorWS_SP00 {class=MaterialExpressionCameraVectorWS, inputs [], properties {}}
  MaterialExpressionComponentMask_0 {class=MaterialExpressionComponentMask, inputs [None {from=MaterialExpressionVectorParameter_1, output=RGB}], properties {g=true, r=true}}
  MaterialExpressionConstant3Vector_0 {class=MaterialExpressionConstant3Vector, properties {constant=<rgba 0.5 0.5 1 0>}}
  MaterialExpressionConstant3Vector_2 {class=MaterialExpressionConstant3Vector, properties {constant=<rgba 0 0 1 0>}}
  MaterialExpressionConstant3Vector_3 {class=MaterialExpressionConstant3Vector, properties {constant=<rgba 0 0 1 0>}}
  MaterialExpressionConstant3Vector_4 {class=MaterialExpressionConstant3Vector, properties {constant=<rgba 0 0 1 0>}}
  MaterialExpressionConstant3Vector_6 {class=MaterialExpressionConstant3Vector, properties {constant=<rgba 0 0 90 0>}}
  MaterialExpressionConstant3Vector_7 {class=MaterialExpressionConstant3Vector}
  MaterialExpressionConstantBiasScale_0 {class=MaterialExpressionConstantBiasScale, inputs [None {from=MaterialExpressionTextureSample_3, output=R}], properties {bias=-0.5, scale=2}}
  MaterialExpressionConstantBiasScale_10 {class=MaterialExpressionConstantBiasScale, inputs [None {from=""}], properties {bias=-0.5, scale=2}}
  MaterialExpressionConstant_0 {class=MaterialExpressionConstant}
  MaterialExpressionConstant_1 {class=MaterialExpressionConstant, properties {r=1}}
  MaterialExpressionConstant_2 {class=MaterialExpressionConstant}
  MaterialExpressionConstant_4 {class=MaterialExpressionConstant, properties {r=1}}
  MaterialExpressionConstant_5 {class=MaterialExpressionConstant}
  MaterialExpressionConvert_0 {
   class=MaterialExpressionConvert
   inputs [
    None {from=MaterialExpressionArccosine_0}
    None {from=MaterialExpressionMaterialFunctionCall_13, output=Vector Converted to Angle}
   ]
   properties {
    nodeName=Make Vector2
    convertInputs [
     {
      type=Scalar
      expressionInput {
       expression=&self:MaterialExpressionArccosine_0
       inputName=None
       mask=0
       maskA=0
       maskB=0
       maskG=0
       maskR=0
       outputIndex=0
      }
     }
     {
      type=Scalar
      expressionInput {
       expression=&self:MaterialExpressionMaterialFunctionCall_13
       inputName=None
       mask=0
       maskA=0
       maskB=0
       maskG=0
       maskR=0
       outputIndex=1
      }
     }
    ]
    convertOutputs [{type=Vector2}]
   }
  }
  MaterialExpressionCrossProduct_1 {
   class=MaterialExpressionCrossProduct
   inputs [
    A {from=MaterialExpressionPixelNormalWS_3, output=""}
    B {from=MaterialExpressionAtmosphericLightVector_4, output=""}
   ]
   properties {}
  }
  MaterialExpressionCustom_1 {
   class=MaterialExpressionCustom
   inputs [
    N {from=MaterialExpressionVertexNormalWS_2, output=""}
    L {from=MaterialExpressionAtmosphericLightVector_2, output=""}
    Ref {from=MaterialExpressionConstant3Vector_2, output=""}
   ]
   properties {
    code="float ndl = dot(N, L);\r\nfloat3 up = abs(dot(L, Ref)) > 0.999 ? float3(0,1,0) : Ref;\r\nfloat3 T = normalize(up - L * dot(up, L));\r\nfloat3 B = cross(L, T);\r\nfloat nT = dot(N, T);\r\nfloat nB = dot(N, B);\r\nfloat v = atan2(nB, nT) * 0.15915494 + 0.5;\r\nreturn float2(ndl, v);"
    description=PolarLightUV
    outputType=CMOT_Float2
    inputs [
     {
      inputName=N
      input {
       expression=&self:MaterialExpressionVertexNormalWS_2
       inputName=None
       mask=0
       maskA=0
       maskB=0
       maskG=0
       maskR=0
       outputIndex=0
      }
     }
     {
      inputName=L
      input {
       expression=&self:MaterialExpressionAtmosphericLightVector_2
       inputName=None
       mask=0
       maskA=0
       maskB=0
       maskG=0
       maskR=0
       outputIndex=0
      }
     }
     {
      inputName=Ref
      input {
       expression=&self:MaterialExpressionConstant3Vector_2
       inputName=None
       mask=1
       maskA=0
       maskB=1
       maskG=1
       maskR=1
       outputIndex=0
      }
     }
    ]
   }
  }
  MaterialExpressionCustom_2 {
   class=MaterialExpressionCustom
   inputs [
    N {from=MaterialExpressionPixelNormalWS_2, output=""}
    L {from=MaterialExpressionAtmosphericLightVector_3, output=""}
    Ref {from=MaterialExpressionConstant3Vector_4, output=""}
    RelPos {from=MaterialExpressionSubtract_4, output=""}
   ]
   properties {
    code="float3 D = normalize(RelPos);\r\nfloat ndl = dot(N, L);\r\nfloat3 up = abs(dot(L, Ref)) > 0.999 ? float3(0,1,0) : Ref;\r\nfloat3 T = normalize(up - L * dot(up, L));\r\nfloat3 B = cross(L, T);\r\nfloat v = atan2(dot(D, B), dot(D, T)) * 0.15915494 + 0.5;\r\nreturn float2(ndl, v);"
    description=PolarLightUV_Char
    outputType=CMOT_Float2
    inputs [
     {
      inputName=N
      input {
       expression=&self:MaterialExpressionPixelNormalWS_2
       inputName=None
       mask=0
       maskA=0
       maskB=0
       maskG=0
       maskR=0
       outputIndex=0
      }
     }
     {
      inputName=L
      input {
       expression=&self:MaterialExpressionAtmosphericLightVector_3
       inputName=None
       mask=0
       maskA=0
       maskB=0
       maskG=0
       maskR=0
       outputIndex=0
      }
     }
     {
      inputName=Ref
      input {
       expression=&self:MaterialExpressionConstant3Vector_4
       inputName=None
       mask=1
       maskA=0
       maskB=1
       maskG=1
       maskR=1
       outputIndex=0
      }
     }
     {
      inputName=RelPos
      input {
       expression=&self:MaterialExpressionSubtract_4
       inputName=None
       mask=0
       maskA=0
       maskB=0
       maskG=0
       maskR=0
       outputIndex=0
      }
     }
    ]
   }
  }
  MaterialExpressionDDX_0 {class=MaterialExpressionDDX, inputs [Value {from=MaterialExpressionNamedRerouteUsage_0, output=""}], properties {}}
  MaterialExpressionDDX_1 {class=MaterialExpressionDDX, inputs [Value {from=MaterialExpressionTextureCoordinate_5, output=""}], properties {}}
  MaterialExpressionDDX_2 {class=MaterialExpressionDDX, inputs [Value {from=MaterialExpressionTextureCoordinate_0, output=""}], properties {}}
  MaterialExpressionDDY_0 {class=MaterialExpressionDDY, inputs [Value {from=MaterialExpressionNamedRerouteUsage_0, output=""}], properties {}}
  MaterialExpressionDDY_1 {class=MaterialExpressionDDY, inputs [Value {from=MaterialExpressionTextureCoordinate_5, output=""}], properties {}}
  MaterialExpressionDDY_2 {class=MaterialExpressionDDY, inputs [Value {from=MaterialExpressionTextureCoordinate_0, output=""}], properties {}}
  MaterialExpressionDivide_0 {
   class=MaterialExpressionDivide
   inputs [
    A {from=MaterialExpressionFloor_0}
    B {from=MaterialExpressionScalarParameter_35}
   ]
  }
  MaterialExpressionDivide_1 {
   class=MaterialExpressionDivide
   inputs [
    A {from=MaterialExpressionFloor_1}
    B {from=MaterialExpressionLinearInterpolate_2}
   ]
   properties {constB=1280}
  }
  MaterialExpressionDivide_2 {
   class=MaterialExpressionDivide
   inputs [
    A {from=MaterialExpressionFloor_2}
    B {from=MaterialExpressionScalarParameter_30}
   ]
  }
  MaterialExpressionDivide_3 {
   class=MaterialExpressionDivide
   inputs [
    A {from=MaterialExpressionFloor_3}
    B {from=MaterialExpressionScalarParameter_7}
   ]
  }
  MaterialExpressionDivide_4 {
   class=MaterialExpressionDivide
   inputs [
    A {from=MaterialExpressionConstantBiasScale_10}
    B {from=MaterialExpressionMultiply_16}
   ]
  }
  MaterialExpressionDivide_6 {
   class=MaterialExpressionDivide
   inputs [
    A {from=MaterialExpressionScalarParameter_10}
    B {from=""}
   ]
   properties {constB=4}
  }
  MaterialExpressionDivide_7 {
   class=MaterialExpressionDivide
   inputs [
    A {from=MaterialExpressionFloor_6}
    B {from=MaterialExpressionScalarParameter_27}
   ]
  }
  MaterialExpressionDivide_8 {
   class=MaterialExpressionDivide
   inputs [
    A {from=MaterialExpressionFloor_4}
    B {from=MaterialExpressionSubtract_3}
   ]
  }
  MaterialExpressionDotProduct_0 {
   class=MaterialExpressionDotProduct
   inputs [
    A {from=MaterialExpressionTransform_2, output=""}
    B {from=MaterialExpressionTransform_1, output=""}
   ]
   properties {}
  }
  MaterialExpressionDotProduct_1 {
   class=MaterialExpressionDotProduct
   inputs [
    A {from=MaterialExpressionTransform_0, output=""}
    B {from=MaterialExpressionMultiply_12, output=""}
   ]
   properties {}
  }
  MaterialExpressionDotProduct_4 {
   class=MaterialExpressionDotProduct
   inputs [
    A {from=MaterialExpressionPixelNormalWS_3, output=""}
    B {from=MaterialExpressionAtmosphericLightVector_4, output=""}
   ]
   properties {}
  }
  MaterialExpressionDotProduct_5 {
   class=MaterialExpressionDotProduct
   inputs [
    A {from=MaterialExpressionSkyAtmosphereLightDirection_1, output=""}
    B {from=MaterialExpressionTransform_1, output=""}
   ]
   properties {}
  }
  MaterialExpressionDotProduct_SP00 {
   class=MaterialExpressionDotProduct
   inputs [
    A {from=MaterialExpressionNormalize_SP00, output=""}
    B {from=MaterialExpressionTransform_1, output=""}
   ]
   properties {}
  }
  MaterialExpressionFloor_0 {class=MaterialExpressionFloor, inputs [None {from=MaterialExpressionMultiply_0, output=""}], properties {}}
  MaterialExpressionFloor_1 {class=MaterialExpressionFloor, inputs [None {from=MaterialExpressionMultiply_2, output=""}], properties {}}
  MaterialExpressionFloor_2 {class=MaterialExpressionFloor, inputs [None {from=MaterialExpressionMultiply_1, output=""}], properties {}}
  MaterialExpressionFloor_3 {class=MaterialExpressionFloor, inputs [None {from=MaterialExpressionMultiply_10, output=""}], properties {}}
  MaterialExpressionFloor_4 {class=MaterialExpressionFloor, inputs [None {from=MaterialExpressionMultiply_15, output=""}], properties {}}
  MaterialExpressionFloor_6 {class=MaterialExpressionFloor, inputs [None {from=MaterialExpressionMultiply_20, output=""}], properties {}}
  MaterialExpressionGetMaterialAttributes_0 {
   class=MaterialExpressionGetMaterialAttributes
   inputs [None {from=MaterialExpressionNamedRerouteUsage_9, output=""}]
   properties {attributeGetTypes [#1, #2, #3]}
  }
  MaterialExpressionIf_0 {
   class=MaterialExpressionIf
   inputs [
    A {from=MaterialExpressionNamedRerouteUsage_0, output=""}
    B {from=MaterialExpressionTextureSample_4, output=R}
    "A > B" {from=MaterialExpressionConstant_4, output=""}
    "A == B" {from="", output=""}
    "A < B" {from=MaterialExpressionConstant_5, output=""}
   ]
   properties {}
  }
  MaterialExpressionLinearInterpolate_0 {
   class=MaterialExpressionLinearInterpolate
   inputs [
    A {from=MaterialExpressionLinearInterpolate_3}
    B {from=MaterialExpressionMultiply_8}
    Alpha {from=MaterialExpressionMultiply_11}
   ]
  }
  MaterialExpressionLinearInterpolate_1 {
   class=MaterialExpressionLinearInterpolate
   inputs [
    A {from=""}
    B {from=MaterialExpressionSubtract_1}
    Alpha {from=MaterialExpressionScalarParameter_1}
   ]
  }
  MaterialExpressionLinearInterpolate_2 {
   class=MaterialExpressionLinearInterpolate
   inputs [
    A {from=MaterialExpressionScalarParameter_0}
    B {from=MaterialExpressionScalarParameter_21}
    Alpha {from=MaterialExpressionMaterialFunctionCall_11, output=Result}
   ]
  }
  MaterialExpressionLinearInterpolate_3 {
   class=MaterialExpressionLinearInterpolate
   inputs [
    A {from=MaterialExpressionMultiply_6}
    B {from=MaterialExpressionMax_2}
    Alpha {from=MaterialExpressionNamedRerouteUsage_3}
   ]
  }
  MaterialExpressionLinearInterpolate_4 {
   class=MaterialExpressionLinearInterpolate
   inputs [
    A {from=""}
    B {from=MaterialExpressionScalarParameter_18}
    Alpha {from=MaterialExpressionNamedRerouteUsage_8}
   ]
  }
  MaterialExpressionLinearInterpolate_5 {
   class=MaterialExpressionLinearInterpolate
   inputs [
    A {from=MaterialExpressionScalarParameter_6}
    B {from=MaterialExpressionScalarParameter_17}
    Alpha {from=MaterialExpressionNamedRerouteUsage_4}
   ]
  }
  MaterialExpressionLinearInterpolate_6 {
   class=MaterialExpressionLinearInterpolate
   inputs [
    A {from=MaterialExpressionLinearInterpolate_0}
    B {from=""}
    Alpha {from=MaterialExpressionNamedRerouteUsage_6}
   ]
   properties {constB=0}
  }
  MaterialExpressionMaterialFunctionCall_0 {
   class=MaterialExpressionMaterialFunctionCall
   inputs [
    Result {from=""}
    Tech Art Mask {from=""}
   ]
   properties {materialFunction=/Game/MF_RimLight}
  }
  MaterialExpressionMaterialFunctionCall_1 {
   class=MaterialExpressionMaterialFunctionCall
   inputs [
    TechArt Mask {from=MaterialExpressionNamedRerouteUsage_10}
    Material {from=MaterialExpressionNamedRerouteDeclaration_5}
   ]
   properties {materialFunction=&/MF_Pixelation_v1}
  }
  MaterialExpressionMaterialFunctionCall_11 {
   class=MaterialExpressionMaterialFunctionCall
   inputs [
    Fade Length {from=""}
    Fade Offset {from=""}
    For Vertex Shader {from=""}
   ]
   properties {materialFunction=&CameraDepthFade}
  }
  MaterialExpressionMaterialFunctionCall_13 {
   class=MaterialExpressionMaterialFunctionCall
   inputs [
    Vector or UVs {from=MaterialExpressionNormalize_0}
    Swizzle Coordinate Output {from=""}
   ]
   properties {materialFunction=&VectorToRadialValue}
  }
  MaterialExpressionMaterialFunctionCall_14 {
   class=MaterialExpressionMaterialFunctionCall
   inputs [
    UVs {from=MaterialExpressionConvert_0}
    Tiling {from=MaterialExpressionScalarParameter_23}
   ]
   properties {materialFunction=/Engine/ArtTools/RenderToTexture/MaterialFunctions/CheckerPattern}
  }
  MaterialExpressionMaterialFunctionCall_15 {
   class=MaterialExpressionMaterialFunctionCall
   inputs [
    Vector to Transform {from=MaterialExpressionCrossProduct_1}
    Z Vector {from=MaterialExpressionAtmosphericLightVector_4}
    Center Location {from=MaterialExpressionConstant3Vector_7}
   ]
   properties {materialFunction=/Engine/Functions/Engine_MaterialFunctions02/Math/TransformToZVector}
  }
  MaterialExpressionMaterialFunctionCall_2 {
   class=MaterialExpressionMaterialFunctionCall
   inputs [
    Fade Length {from=MaterialExpressionScalarParameter_8}
    Fade Offset {from=MaterialExpressionScalarParameter_16}
    For Vertex Shader {from=""}
   ]
   properties {materialFunction=&CameraDepthFade}
  }
  MaterialExpressionMaterialFunctionCall_3 {
   class=MaterialExpressionMaterialFunctionCall
   inputs [
    Material {from=MaterialExpressionMaterialFunctionCall_1, output=Result}
    M {from=MaterialExpressionMaterialFunctionCall_4, output=M}
    Pre Skinned Local Position {from=MaterialExpressionMaterialFunctionCall_4, output=M}
    Distance Blend {from=MaterialExpressionMaterialFunctionCall_4, output=M}
    Fresnel {from=MaterialExpressionMaterialFunctionCall_4, output=M}
   ]
   properties {materialFunction=&MF_CharacterParent_2023_PostFX}
  }
  MaterialExpressionMaterialFunctionCall_4 {
   class=MaterialExpressionMaterialFunctionCall
   inputs [
    Overide Texture Set UVs {from=""}
    Override Tech Art Mask UVs {from=""}
    Override Basic Emissive UVs {from=""}
   ]
   properties {materialFunction=&MF_CharacterParent_2023_PreFX}
  }
  MaterialExpressionMaterialFunctionCall_5 {
   class=MaterialExpressionMaterialFunctionCall
   inputs [
    In {from=MaterialExpressionMaterialFunctionCall_4, output=Result}
    Mask Input {from=MaterialExpressionMaterialFunctionCall_4, output=Result}
   ]
   properties {materialFunction=/Game/Global/Materials/Character_Parent_2023/Hair_Parent_2023/MaterialFunctions/MF_FN_Hair_2023}
  }
  MaterialExpressionMaterialFunctionCall_6 {
   class=MaterialExpressionMaterialFunctionCall
   inputs [MaterialInput {from=MaterialExpressionBlendMaterialAttributes_1}]
   properties {materialFunction=/Game/Global/Materials/Character_Parent_2023/Hair_Parent_2023/MaterialFunctions/MF_FlyAwayHair_2023}
  }
  MaterialExpressionMaterialFunctionCall_9 {
   class=MaterialExpressionMaterialFunctionCall
   inputs [
    Tile {from=""}
    DistortionAdjustment {from=""}
   ]
   properties {materialFunction=/Game/Global/Materials/MaterialGym/Materials/MaterialElements/UVs/CylindricalProjection/MF_CylindricalProjUV_Mod}
  }
  MaterialExpressionMax_0 {
   class=MaterialExpressionMax
   inputs [
    A {from=MaterialExpressionSaturate_2}
    B {from=MaterialExpressionSaturate_4}
   ]
  }
  MaterialExpressionMax_1 {
   class=MaterialExpressionMax
   inputs [
    A {from=MaterialExpressionDivide_7}
    B {from=MaterialExpressionScalarParameter_28}
   ]
  }
  MaterialExpressionMax_2 {
   class=MaterialExpressionMax
   inputs [
    A {from=MaterialExpressionDivide_3}
    B {from=MaterialExpressionMultiply_5}
   ]
  }
  MaterialExpressionMax_3 {
   class=MaterialExpressionMax
   inputs [
    A {from=MaterialExpressionTextureSampleParameter2D_3, output=R}
    B {from=MaterialExpressionTextureSampleParameter2D_3, output=R}
   ]
  }
  MaterialExpressionMax_4 {
   class=MaterialExpressionMax
   inputs [
    A {from=MaterialExpressionMax_3}
    B {from=MaterialExpressionTextureSampleParameter2D_3, output=B}
   ]
  }
  MaterialExpressionMin_0 {
   class=MaterialExpressionMin
   inputs [
    A {from=MaterialExpressionSaturate_2, output=""}
    B {from=MaterialExpressionSaturate_4, output=""}
   ]
   properties {}
  }
  MaterialExpressionMultiply_0 {
   class=MaterialExpressionMultiply
   inputs [
    A {from=MaterialExpressionTextureCoordinate_0}
    B {from=MaterialExpressionScalarParameter_35}
   ]
  }
  MaterialExpressionMultiply_1 {
   class=MaterialExpressionMultiply
   inputs [
    A {from=MaterialExpressionPower_SP00}
    B {from=MaterialExpressionScalarParameter_30}
   ]
  }
  MaterialExpressionMultiply_10 {
   class=MaterialExpressionMultiply
   inputs [
    A {from=MaterialExpressionMultiply_13}
    B {from=MaterialExpressionScalarParameter_7}
   ]
  }
  MaterialExpressionMultiply_11 {
   class=MaterialExpressionMultiply
   inputs [
    A {from=MaterialExpressionMultiply_7}
    B {from=MaterialExpressionMultiply_27}
   ]
  }
  MaterialExpressionMultiply_12 {
   class=MaterialExpressionMultiply
   inputs [
    A {from=MaterialExpressionTransform_1}
    B {from=MaterialExpressionScalarParameter_31}
   ]
  }
  MaterialExpressionMultiply_13 {
   class=MaterialExpressionMultiply
   inputs [
    A {from=MaterialExpressionTextureSampleParameter2D_0, output=RGB}
    B {from=MaterialExpressionScalarParameter_9}
   ]
  }
  MaterialExpressionMultiply_14 {
   class=MaterialExpressionMultiply
   inputs [
    A {from=MaterialExpressionTextureCoordinate_4}
    B {from=MaterialExpressionDivide_6}
   ]
  }
  MaterialExpressionMultiply_15 {
   class=MaterialExpressionMultiply
   inputs [
    A {from=MaterialExpressionAdd_3}
    B {from=MaterialExpressionScalarParameter_12}
   ]
  }
  MaterialExpressionMultiply_16 {
   class=MaterialExpressionMultiply
   inputs [
    A {from=MaterialExpressionScalarParameter_25}
    B {from=MaterialExpressionScalarParameter_14}
   ]
   properties {constB=2}
  }
  MaterialExpressionMultiply_17 {
   class=MaterialExpressionMultiply
   inputs [
    A {from=MaterialExpressionNamedRerouteUsage_0}
    B {from=MaterialExpressionScalarParameter_19}
   ]
  }
  MaterialExpressionMultiply_18 {
   class=MaterialExpressionMultiply
   inputs [
    A {from=MaterialExpressionCustom_2}
    B {from=MaterialExpressionScalarParameter_20}
   ]
  }
  MaterialExpressionMultiply_2 {
   class=MaterialExpressionMultiply
   inputs [
    A {from=MaterialExpressionTransformPosition_0}
    B {from=MaterialExpressionLinearInterpolate_2}
   ]
   properties {constB=1280}
  }
  MaterialExpressionMultiply_20 {
   class=MaterialExpressionMultiply
   inputs [
    A {from=MaterialExpressionPower_4}
    B {from=MaterialExpressionScalarParameter_27}
   ]
  }
  MaterialExpressionMultiply_23 {
   class=MaterialExpressionMultiply
   inputs [
    A {from=MaterialExpressionConstantBiasScale_0}
    B {from=MaterialExpressionScalarParameter_29}
   ]
  }
  MaterialExpressionMultiply_27 {
   class=MaterialExpressionMultiply
   inputs [
    A {from=MaterialExpressionScalarParameter_18}
    B {from=MaterialExpressionStep_5}
   ]
  }
  MaterialExpressionMultiply_4 {
   class=MaterialExpressionMultiply
   inputs [
    A {from=MaterialExpressionTextureCoordinate_2}
    B {from=MaterialExpressionScalarParameter_35}
   ]
  }
  MaterialExpressionMultiply_5 {
   class=MaterialExpressionMultiply
   inputs [
    A {from=MaterialExpressionTextureSampleParameter2D_3, output=RGB}
    B {from=MaterialExpressionScalarParameter_4}
   ]
  }
  MaterialExpressionMultiply_6 {
   class=MaterialExpressionMultiply
   inputs [
    A {from=MaterialExpressionMax_2}
    B {from=MaterialExpressionNamedRerouteDeclaration_8}
   ]
  }
  MaterialExpressionMultiply_7 {
   class=MaterialExpressionMultiply
   inputs [
    A {from=MaterialExpressionDivide_2}
    B {from=MaterialExpressionScalarParameter_15}
   ]
  }
  MaterialExpressionMultiply_8 {
   class=MaterialExpressionMultiply
   inputs [
    A {from=MaterialExpressionSkyAtmosphereLightIlluminance_1}
    B {from=MaterialExpressionMultiply_7}
   ]
  }
  MaterialExpressionMultiply_9 {
   class=MaterialExpressionMultiply
   inputs [
    A {from=""}
    B {from=MaterialExpressionLinearInterpolate_5}
   ]
  }
  MaterialExpressionNamedRerouteDeclaration_0 {
   class=MaterialExpressionNamedRerouteDeclaration
   inputs [None {from=MaterialExpressionGetMaterialAttributes_0, output=Normal}]
   properties {name=IN_Normal, nodeColor=<rgba 1 0.2344 0 1>}
  }
  MaterialExpressionNamedRerouteDeclaration_1 {
   class=MaterialExpressionNamedRerouteDeclaration
   inputs [None {from=MaterialExpressionAdd_3}]
   properties {name=W, nodeColor=<rgba 0 1 0.4609 1>}
  }
  MaterialExpressionNamedRerouteDeclaration_10 {
   class=MaterialExpressionNamedRerouteDeclaration
   inputs [None {from=MaterialExpressionMax_4}]
   properties {name=Emissive Mask, nodeColor=<rgba 1 0.3516 0 1>}
  }
  MaterialExpressionNamedRerouteDeclaration_11 {
   class=MaterialExpressionNamedRerouteDeclaration
   inputs [None {from=MaterialExpressionSetMaterialAttributes_0}]
   properties {name=OUT_Mat, nodeColor=<rgba 0.3594 1 0 1>}
  }
  MaterialExpressionNamedRerouteDeclaration_12 {
   class=MaterialExpressionNamedRerouteDeclaration
   inputs [None {from=MaterialExpressionMaterialFunctionCall_4, output=Tech Art Mask}]
   properties {name=TA Mask, nodeColor=<rgba 0 0.7422 1 1>}
  }
  MaterialExpressionNamedRerouteDeclaration_2 {
   class=MaterialExpressionNamedRerouteDeclaration
   inputs [None {from=MaterialExpressionMaterialFunctionCall_2, output=Result}]
   properties {name=CameraDistanceFade, nodeColor=<rgba 0.5234 1 0 1>}
  }
  MaterialExpressionNamedRerouteDeclaration_3 {
   class=MaterialExpressionNamedRerouteDeclaration
   inputs [None {from=MaterialExpressionTextureSample_4, output=R}]
   properties {name=Noise, nodeColor=<rgba 0.4531 0 1 1>}
  }
  MaterialExpressionNamedRerouteDeclaration_4 {
   class=MaterialExpressionNamedRerouteDeclaration
   inputs [None {from=MaterialExpressionMultiply_23}]
   properties {name=Dithered Kernel, nodeColor=<rgba 1 0 0.2109 1>}
  }
  MaterialExpressionNamedRerouteDeclaration_5 {
   class=MaterialExpressionNamedRerouteDeclaration
   inputs [None {from=MaterialExpressionStaticSwitchParameter_0}]
   properties {name=IN_Mat, nodeColor=<rgba 0.7344 1 0 1>}
  }
  MaterialExpressionNamedRerouteDeclaration_6 {
   class=MaterialExpressionNamedRerouteDeclaration
   inputs [None {from=MaterialExpressionDivide_4}]
   properties {name=Lighting Dither, nodeColor=<rgba 0.5 1 0 1>}
  }
  MaterialExpressionNamedRerouteDeclaration_7 {
   class=MaterialExpressionNamedRerouteDeclaration
   inputs [None {from=MaterialExpressionStep_0}]
   properties {name=Outline, nodeColor=<rgba 0.8281 1 0 1>}
  }
  MaterialExpressionNamedRerouteDeclaration_8 {
   class=MaterialExpressionNamedRerouteDeclaration
   inputs [None {from=MaterialExpressionSaturate_0}]
   properties {name=Shadow dither, nodeColor=<rgba 1 0.6562 0 1>}
  }
  MaterialExpressionNamedRerouteDeclaration_9 {
   class=MaterialExpressionNamedRerouteDeclaration
   inputs [None {from=MaterialExpressionGetMaterialAttributes_0, output=Metallic}]
   properties {name=IN_Spec, nodeColor=<rgba 0.5 0 1 1>}
  }
  MaterialExpressionNamedRerouteUsage_0 {class=MaterialExpressionNamedRerouteUsage}
  MaterialExpressionNamedRerouteUsage_1 {class=MaterialExpressionNamedRerouteUsage}
  MaterialExpressionNamedRerouteUsage_10 {class=MaterialExpressionNamedRerouteUsage}
  MaterialExpressionNamedRerouteUsage_2 {class=MaterialExpressionNamedRerouteUsage}
  MaterialExpressionNamedRerouteUsage_3 {class=MaterialExpressionNamedRerouteUsage}
  MaterialExpressionNamedRerouteUsage_4 {class=MaterialExpressionNamedRerouteUsage}
  MaterialExpressionNamedRerouteUsage_5 {class=MaterialExpressionNamedRerouteUsage}
  MaterialExpressionNamedRerouteUsage_6 {class=MaterialExpressionNamedRerouteUsage}
  MaterialExpressionNamedRerouteUsage_7 {class=MaterialExpressionNamedRerouteUsage}
  MaterialExpressionNamedRerouteUsage_8 {class=MaterialExpressionNamedRerouteUsage}
  MaterialExpressionNamedRerouteUsage_9 {class=MaterialExpressionNamedRerouteUsage}
  MaterialExpressionNormalize_0 {class=MaterialExpressionNormalize, inputs [VectorInput {from=MaterialExpressionMaterialFunctionCall_15, output=Result}]}
  MaterialExpressionNormalize_SP00 {class=MaterialExpressionNormalize, inputs [VectorInput {from=MaterialExpressionAdd_SP00, output=""}]}
  MaterialExpressionObjectPositionWS_0 {class=MaterialExpressionObjectPositionWS, inputs [], properties {}}
  MaterialExpressionOneMinus_0 {class=MaterialExpressionOneMinus, inputs [None {from=MaterialExpressionStep_0}]}
  MaterialExpressionPixelNormalWS_1 {class=MaterialExpressionPixelNormalWS, inputs [], properties {}}
  MaterialExpressionPixelNormalWS_2 {class=MaterialExpressionPixelNormalWS, inputs [], properties {}}
  MaterialExpressionPixelNormalWS_3 {class=MaterialExpressionPixelNormalWS, inputs [], properties {}}
  MaterialExpressionPower_4 {
   class=MaterialExpressionPower
   inputs [
    Base {from=MaterialExpressionAdd_10}
    Exp {from=MaterialExpressionScalarParameter_26}
   ]
  }
  MaterialExpressionPower_SP00 {
   class=MaterialExpressionPower
   inputs [
    Base {from=MaterialExpressionSaturate_3}
    Exp {from=MaterialExpressionScalarParameter_SP00}
   ]
  }
  MaterialExpressionReroute_0 {class=MaterialExpressionReroute, inputs [None {from=MaterialExpressionReroute_1}]}
  MaterialExpressionReroute_1 {class=MaterialExpressionReroute, inputs [None {from=MaterialExpressionMaterialFunctionCall_4, output=Result}]}
  MaterialExpressionSaturate_0 {class=MaterialExpressionSaturate, inputs [None {from=MaterialExpressionMax_1}]}
  MaterialExpressionSaturate_1 {class=MaterialExpressionSaturate, inputs [None {from=MaterialExpressionAdd_1}]}
  MaterialExpressionSaturate_2 {class=MaterialExpressionSaturate, inputs [None {from=MaterialExpressionDotProduct_1}]}
  MaterialExpressionSaturate_3 {class=MaterialExpressionSaturate, inputs [None {from=MaterialExpressionDotProduct_SP00}]}
  MaterialExpressionSaturate_4 {class=MaterialExpressionSaturate, inputs [None {from=MaterialExpressionDotProduct_0}]}
  MaterialExpressionScalarParameter_0 {class=MaterialExpressionScalarParameter, properties {defaultValue=4, parameterName=VertexSwimGridSize}}
  MaterialExpressionScalarParameter_1 {class=MaterialExpressionScalarParameter, properties {defaultValue=1, parameterName=Use Vertex Swimming?}}
  MaterialExpressionScalarParameter_10 {class=MaterialExpressionScalarParameter, properties {defaultValue=128, parameterName=Noise Tiling}}
  MaterialExpressionScalarParameter_12 {class=MaterialExpressionScalarParameter, properties {defaultValue=3, parameterName=BandCount}}
  MaterialExpressionScalarParameter_13 {class=MaterialExpressionScalarParameter, properties {parameterName=s}}
  MaterialExpressionScalarParameter_14 {
   class=MaterialExpressionScalarParameter
   properties {defaultValue=5, desc="1 = \"smooth\" gradient. Higher creates bands", parameterName=DitherTightness}
  }
  MaterialExpressionScalarParameter_15 {class=MaterialExpressionScalarParameter, properties {defaultValue=1, parameterName=Spec}}
  MaterialExpressionScalarParameter_16 {
   class=MaterialExpressionScalarParameter
   properties {defaultValue=350, group="CreamDecoy - Rimlights && Highlights", parameterName="Fade Offset (S)"}
  }
  MaterialExpressionScalarParameter_17 {
   class=MaterialExpressionScalarParameter
   properties {defaultValue=0.1, group="CreamDecoy - Rimlights && Highlights|Highlights", parameterName=HighligtsIntensity_Far}
  }
  MaterialExpressionScalarParameter_18 {class=MaterialExpressionScalarParameter, properties {defaultValue=0.05, parameterName=Spec Blend}}
  MaterialExpressionScalarParameter_19 {class=MaterialExpressionScalarParameter}
  MaterialExpressionScalarParameter_20 {class=MaterialExpressionScalarParameter, properties {defaultValue=2.2003, parameterName=Param_1}}
  MaterialExpressionScalarParameter_21 {class=MaterialExpressionScalarParameter, properties {defaultValue=4, parameterName=Vertex Swim Grid Size Max}}
  MaterialExpressionScalarParameter_22 {class=MaterialExpressionScalarParameter, properties {defaultValue=0.5, parameterName=Param_4}}
  MaterialExpressionScalarParameter_23 {class=MaterialExpressionScalarParameter, properties {defaultValue=20, parameterName=Param_2}}
  MaterialExpressionScalarParameter_24 {class=MaterialExpressionScalarParameter, properties {defaultValue=0.7, parameterName=Step Outline}}
  MaterialExpressionScalarParameter_25 {class=MaterialExpressionScalarParameter, properties {defaultValue=3, parameterName=BandCount}}
  MaterialExpressionScalarParameter_26 {class=MaterialExpressionScalarParameter, properties {defaultValue=1, parameterName=Band Curve}}
  MaterialExpressionScalarParameter_27 {class=MaterialExpressionScalarParameter, properties {defaultValue=3, parameterName=BandCount}}
  MaterialExpressionScalarParameter_28 {class=MaterialExpressionScalarParameter, properties {defaultValue=0.2572, parameterName=ShadowMin}}
  MaterialExpressionScalarParameter_29 {class=MaterialExpressionScalarParameter, properties {parameterName=Dither BC Intensity}}
  MaterialExpressionScalarParameter_3 {class=MaterialExpressionScalarParameter, properties {defaultValue=75, parameterName=DitherScale}}
  MaterialExpressionScalarParameter_30 {class=MaterialExpressionScalarParameter, properties {defaultValue=3, parameterName=BandCount}}
  MaterialExpressionScalarParameter_31 {class=MaterialExpressionScalarParameter, properties {defaultValue=3, parameterName=Normal Intensity Boost}}
  MaterialExpressionScalarParameter_32 {class=MaterialExpressionScalarParameter, properties {defaultValue=0.5, parameterName=Band Offset}}
  MaterialExpressionScalarParameter_33 {class=MaterialExpressionScalarParameter, properties {defaultValue=1, parameterName=Param_3}}
  MaterialExpressionScalarParameter_35 {class=MaterialExpressionScalarParameter, properties {defaultValue=250, parameterName=Quantized Scale}}
  MaterialExpressionScalarParameter_4 {class=MaterialExpressionScalarParameter, properties {defaultValue=2, group=06_Emissive, parameterName=EmissiveBrightness}}
  MaterialExpressionScalarParameter_6 {
   class=MaterialExpressionScalarParameter
   properties {defaultValue=0.5, group="CreamDecoy - Rimlights && Highlights|Highlights", parameterName=HighligtsIntensity_Near}
  }
  MaterialExpressionScalarParameter_7 {class=MaterialExpressionScalarParameter, properties {defaultValue=32, parameterName=Steps}}
  MaterialExpressionScalarParameter_8 {
   class=MaterialExpressionScalarParameter
   properties {defaultValue=512, group="CreamDecoy - Rimlights && Highlights", parameterName="Fade Length (S)"}
  }
  MaterialExpressionScalarParameter_9 {class=MaterialExpressionScalarParameter, properties {defaultValue=1.25, parameterName=BC itnensity}}
  MaterialExpressionScalarParameter_SP00 {class=MaterialExpressionScalarParameter, properties {defaultValue=6, parameterName=SpecularSharpness}}
  MaterialExpressionSetMaterialAttributes_0 {
   class=MaterialExpressionSetMaterialAttributes
   inputs [
    MaterialAttributes {from=MaterialExpressionGetMaterialAttributes_0, output=MaterialAttributes}
    Emissive Color {from=MaterialExpressionLinearInterpolate_0, output=""}
    Normal {from=MaterialExpressionConstant3Vector_0, output=""}
    Specular {from=MaterialExpressionConstant_0, output=""}
    Roughness {from=MaterialExpressionConstant_1, output=""}
    Base Color {from=MaterialExpressionConstant_0, output=""}
    Fuzz Color or Subsurface Color {from=MaterialExpressionConstant_2, output=""}
    Opacity {from=MaterialExpressionConstant_2, output=""}
   ]
   properties {attributeSetTypes [
     #4
     #2
     #5
     #6
     #1
     #7
     #8
    ]}
  }
  MaterialExpressionSetMaterialAttributes_1 {
   class=MaterialExpressionSetMaterialAttributes
   inputs [
    MaterialAttributes {from=MaterialExpressionSetMaterialAttributes_0, output=""}
    World Position Offset {from=MaterialExpressionLinearInterpolate_1, output=""}
   ]
   properties {attributeSetTypes [#9]}
  }
  MaterialExpressionSkyAtmosphereLightDirection_1 {class=MaterialExpressionSkyAtmosphereLightDirection, inputs [], properties {}}
  MaterialExpressionSkyAtmosphereLightDirection_SP00 {class=MaterialExpressionSkyAtmosphereLightDirection, inputs [], properties {}}
  MaterialExpressionSkyAtmosphereLightIlluminance_1 {class=MaterialExpressionSkyAtmosphereLightIlluminance, inputs [World Position {from="", output=""}], properties {}}
  MaterialExpressionStaticSwitchParameter_0 {
   class=MaterialExpressionStaticSwitchParameter
   inputs [
    True {from=MaterialExpressionMaterialFunctionCall_6, output=Result}
    False {from=MaterialExpressionBlendMaterialAttributes_1, output=""}
   ]
   properties {group=Skin FX, parameterName=IsFlyAwayHair?}
  }
  MaterialExpressionStaticSwitchParameter_60 {
   class=MaterialExpressionStaticSwitchParameter
   inputs [
    True {from=MaterialExpressionMaterialFunctionCall_5, output=Result}
    False {from=MaterialExpressionMaterialFunctionCall_4, output=Result}
   ]
   properties {group=Skin FX, parameterName=UseFNHair2_0}
  }
  MaterialExpressionStep_0 {
   class=MaterialExpressionStep
   inputs [
    Y {from=MaterialExpressionSaturate_1, output=""}
    X {from=MaterialExpressionScalarParameter_24, output=""}
   ]
   properties {}
  }
  MaterialExpressionStep_2 {
   class=MaterialExpressionStep
   inputs [
    Y {from=MaterialExpressionNamedRerouteUsage_1, output=""}
    X {from=MaterialExpressionScalarParameter_13, output=""}
   ]
   properties {}
  }
  MaterialExpressionStep_5 {
   class=MaterialExpressionStep
   inputs [
    Y {from=MaterialExpressionScalarParameter_33, output=""}
    X {from=MaterialExpressionNamedRerouteUsage_7, output=""}
   ]
   properties {}
  }
  MaterialExpressionSubtract_1 {
   class=MaterialExpressionSubtract
   inputs [
    A {from=MaterialExpressionTransformPosition_1}
    B {from=MaterialExpressionWorldPosition_0, output=XYZ}
   ]
  }
  MaterialExpressionSubtract_2 {
   class=MaterialExpressionSubtract
   inputs [
    A {from=""}
    B {from=MaterialExpressionNamedRerouteUsage_1}
   ]
  }
  MaterialExpressionSubtract_3 {
   class=MaterialExpressionSubtract
   inputs [
    A {from=MaterialExpressionScalarParameter_12}
    B {from=""}
   ]
  }
  MaterialExpressionSubtract_4 {
   class=MaterialExpressionSubtract
   inputs [
    A {from=MaterialExpressionWorldPosition_1, output=XYZ}
    B {from=MaterialExpressionAdd_5}
   ]
  }
  MaterialExpressionTextureCoordinate_0 {class=MaterialExpressionTextureCoordinate}
  MaterialExpressionTextureCoordinate_2 {class=MaterialExpressionTextureCoordinate}
  MaterialExpressionTextureCoordinate_4 {class=MaterialExpressionTextureCoordinate}
  MaterialExpressionTextureCoordinate_5 {class=MaterialExpressionTextureCoordinate}
  MaterialExpressionTextureSampleParameter2D_0 {
   class=MaterialExpressionTextureSampleParameter2D
   inputs [
    UVs {from=MaterialExpressionDivide_0, output=""}
    "DDX(UVs)" {from=MaterialExpressionDDX_2, output=""}
    "DDY(UVs)" {from=MaterialExpressionDDY_2, output=""}
    Apply View MipBias {from="", output=""}
   ]
   properties {
    group=00_Texture Inputs
    mipValueMode=TMVM_Derivative
    parameterName=Diffuse
    samplerSource=SSM_Wrap_WorldGroupSettings
    sortPriority=0
    texture=/BRCosmetics/Characters/Player/Female/Medium/Bodies/F_MED_PhoenixKick_Gain/Textures/T_PhoenixKick_Gain_Body_D
   }
  }
  MaterialExpressionTextureSampleParameter2D_1 {
   class=MaterialExpressionTextureSampleParameter2D
   inputs [
    UVs {from=MaterialExpressionTextureCoordinate_5, output=""}
    "DDX(UVs)" {from=MaterialExpressionDDX_1, output=""}
    "DDY(UVs)" {from=MaterialExpressionDDY_1, output=""}
    Apply View MipBias {from="", output=""}
   ]
   properties {
    mipValueMode=TMVM_Derivative
    parameterName=Pixelated Normals
    samplerType=SAMPLERTYPE_Normal
    texture=&FlatNormalBlue
   }
  }
  MaterialExpressionTextureSampleParameter2D_3 {
   class=MaterialExpressionTextureSampleParameter2D
   inputs [
    UVs {from=MaterialExpressionDivide_0, output=""}
    Apply View MipBias {from="", output=""}
   ]
   properties {group=06_Emissive, parameterName=Emissive, texture=/EpicBaseTextures/Blanks/T_Default_EmissiveColor_Black}
  }
  MaterialExpressionTextureSample_3 {
   class=MaterialExpressionTextureSample
   inputs [
    UVs {from=MaterialExpressionMultiply_4}
    Tex {from=""}
    Apply View MipBias {from=""}
   ]
   properties {samplerType=SAMPLERTYPE_Grayscale}
  }
  MaterialExpressionTextureSample_4 {
   class=MaterialExpressionTextureSample
   inputs [
    UVs {from=MaterialExpressionMultiply_18}
    Tex {from=""}
    Apply View MipBias {from=""}
   ]
   properties {samplerType=SAMPLERTYPE_Grayscale}
  }
  MaterialExpressionTransformPosition_0 {
   class=MaterialExpressionTransformPosition
   inputs [None {from=MaterialExpressionWorldPosition_0, output=XYZ}]
   properties {transformSourceType=TRANSFORMPOSSOURCE_TranslatedWorld, transformType=TRANSFORMPOSSOURCE_View}
  }
  MaterialExpressionTransformPosition_1 {
   class=MaterialExpressionTransformPosition
   inputs [None {from=MaterialExpressionDivide_1, output=""}]
   properties {transformSourceType=TRANSFORMPOSSOURCE_View, transformType=TRANSFORMPOSSOURCE_TranslatedWorld}
  }
  MaterialExpressionTransform_0 {
   class=MaterialExpressionTransform
   inputs [None {from=MaterialExpressionVectorParameter_2, output=RGB}]
   properties {transformSourceType=TRANSFORMSOURCE_View}
  }
  MaterialExpressionTransform_1 {class=MaterialExpressionTransform, inputs [None {from=MaterialExpressionTextureSampleParameter2D_1, output=RGB}]}
  MaterialExpressionTransform_2 {
   class=MaterialExpressionTransform
   inputs [None {from=MaterialExpressionVectorParameter_0, output=RGB}]
   properties {transformSourceType=TRANSFORMSOURCE_View}
  }
  MaterialExpressionVectorParameter_0 {
   class=MaterialExpressionVectorParameter
   properties {group="CreamDecoy - Rimlights && Highlights|Highlights", parameterName=2222, defaultValue=<rgba 0.5 0 -1 1>}
  }
  MaterialExpressionVectorParameter_1 {class=MaterialExpressionVectorParameter, properties {parameterName=v, defaultValue=<rgba 0.8919 8.2276 0 0>}}
  MaterialExpressionVectorParameter_2 {
   class=MaterialExpressionVectorParameter
   properties {group="CreamDecoy - Rimlights && Highlights|Highlights", parameterName=HighlightsScreenSpacePosition, defaultValue=<rgba 0 0 -1 1>}
  }
  MaterialExpressionVertexNormalWS_2 {class=MaterialExpressionVertexNormalWS, inputs [], properties {}}
  MaterialExpressionViewProperty_0 {class=MaterialExpressionViewProperty, inputs [], properties {}}
  MaterialExpressionWorldPosition_0 {class=MaterialExpressionWorldPosition, inputs [], properties {worldPositionShaderOffset=WPT_CameraRelative}}
  MaterialExpressionWorldPosition_1 {class=MaterialExpressionWorldPosition, inputs [], properties {}}
 }
 properties {
  bAutomaticallySetUsageInEditor=false
  bUseMaterialAttributes=true
  bUsedWithMeshDeformer=true
  bUsedWithMorphTargets=true
  bUsedWithSkeletalMesh=true
  pixelDepthOffsetMode=PDOM_Legacy
  previewMesh=/BRCosmetics/Characters/Player/Female/Medium/Bodies/F_MED_PhoenixKick_Gain/Meshes/F_MED_PhoenixKick_Gain
  shadingModel=MSM_FromMaterialExpression
  usedShadingModels=Default Lit | Subsurface | Clear Coat | Subsurface Profile | Cloth
 }
}
