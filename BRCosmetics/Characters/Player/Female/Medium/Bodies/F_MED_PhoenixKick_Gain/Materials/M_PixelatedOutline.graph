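# Translucent unlit pixelated outline material for the F_MED_PhoenixKick_Gain character body.
# Snaps screen UVs to a jitter-compensated, distance-scaled pixel grid, comparing pixel depth against four diagonal CustomDepth samples per cell to derive edge opacity through MF_SortTranslucency.
# Color comes from UV-quantized, posterized diffuse, shaded by sky light direction with optional Z-lock and modulated by scanlines.
# Maskable world position offset along normals scales with distance.
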Material /BRCosmetics/Characters/Player/Female/Medium/Bodies/F_MED_PhoenixKick_Gain/Materials/M_PixelatedOutline
graph {
 properties {
  bAutomaticallySetUsageInEditor=false
  bEnableResponsiveAA=true
  bUsedWithSkeletalMesh=true
  blendMode=BLEND_Translucent
  previewMesh=/BRCosmetics/Characters/Player/Female/Medium/Bodies/F_MED_PhoenixKick_Gain/Meshes/F_MED_PhoenixKick_Gain
  shadingModel=MSM_Unlit
  usedShadingModels=Unlit
 }
 nodes [
  MaterialExpressionSubtract_2 = Subtract(A=CameraPositionWS(), B=ObjectPositionWS())
  n1 = Custom!(x=@MaterialExpressionSubtract_2) {
   code="length(x)"
   inputs [{
     inputName=x
     input {expression=&self:MaterialExpressionSubtract_2}
    }]
  }
  n2 = Constant3Vector() {constant=<rgba 0 0 -1 0>}
  n3 = Multiply(
   A=@n1
   B=DotProduct(A=Normalize(VectorInput=@MaterialExpressionSubtract_2), B=Transform(None=@n2) {transformSourceType=TRANSFORMSOURCE_View})
  )
  Add(A=@n3)
  n4 = ScreenPosition()
  n5 = Divide(A=ViewProperty() {property=MEVP_TemporalSampleOffset}.Property, B=ViewSize())
  n6 = ScalarParameter() {defaultValue=1000, group=Pixelation - Scanlines, parameterName=Pixelation.ScanlinesDistanceScaling_ScaleFar, sliderMax=2, sortPriority=5}
  n7 = ScalarParameter() {defaultValue=300, group=Pixelation - Scanlines, parameterName=Pixelation.ScanlinesDistanceScaling_StartDistance, sliderMax=2, sortPriority=5}
  n8 = ScalarParameter() {defaultValue=2000, group=Pixelation - Scanlines, parameterName=Pixelation.Scanlines_DistanceFadeLength, sliderMax=2, sortPriority=5}
  n9 = Saturate(None=Divide(A=Subtract(A=@n3, B=@n7), B=@n8))
  n10 = Saturate(None=Divide(A=@n3, B=@n7))
  n11 = LinearInterpolate(
   A=Constant() {r=150}
   B=LinearInterpolate(A=ScalarParameter() {defaultValue=300, parameterName=Pixel Amount B}, B=@n6, Alpha=@n9)
   Alpha=@n10
  )
  n12 = Multiply(A=Subtract(A=@n11, B=Frac(None=@n11)), B=ScalarParameter() {defaultValue=1, parameterName=PixelScale})
  n13 = ViewSize()
  n14 = ComponentMask(None=@n13) {r=true}
  n15 = ComponentMask(None=@n13) {g=true}
  n16 = MaterialFunctionCall(X=Multiply(A=@n12, B=Divide(A=@n14, B=@n15)), Y=@n12) {materialFunction=&MakeFloat2}
  n17 = Floor(None=Multiply(A=Subtract(A=@n4.ViewportUV, B=@n5), B=@n16.Result) {constB=100})
  Add(A=@n17) {constB=0.5}
  Constant() {r=1}
  Constant()
  Constant() {r=1}
  Constant()
  n18 = TextureSampleParameter2D() {parameterName=FX, samplerType=SAMPLERTYPE_Masks, texture=/EpicBaseTextures/Blanks/T_Default_Masks_AllZero}
  n19 = ScalarParameter() {parameterName=UseWPOMask}
  n20 = Saturate(None=Add(A=ChannelMaskParameter(None=@n18.RGB) {parameterName=WPO Mask}.R, B=OneMinus(None=@n19)))
  n21 = LinearInterpolate(
   A=ScalarParameter() {defaultValue=1, parameterName=WPO}
   Alpha=Multiply(A=@n9, B=ScalarParameter() {defaultValue=1, parameterName=UseDistanceScaling})
  ) {constB=4}
  LinearInterpolate(
   A=Subtract(A=Constant(), B=WorldPosition().XYZ)
   B=Multiply(A=VertexNormalWS(), B=Multiply(A=@n20, B=@n21))
   Alpha=ScalarParameter() {defaultValue=1, parameterName=Pixelated}
  )
  MaterialFunctionCall() {materialFunction=&ObjectPivotPoint}
  n22 = PixelDepth()
  n23 = Divide(B=@n16.Result) {constA=0.5, desc=Half Axis Length of One Square}
  n24 = Add(A=Divide(A=@n17, B=@n16.Result), B=@n23) {desc=Center of The Square}
  n25 = Multiply(A=@n23, B=ScalarParameter() {defaultValue=1, parameterName=Kernel Size}) {constB=1.8}
  n26 = Add(A=@n24, B=Multiply(A=@n25, B=Constant2Vector() {g=1, r=1}))
  n27 = LinearInterpolate(A=ScalarParameter() {defaultValue=3, parameterName=Depth Tolerance Min}, B=Constant() {r=20}, Alpha=@n9)
  n28 = SphereMask(A=@n22.R, B=ComponentMask(None=SceneTexture(UVs=@n26) {bFiltered=true, sceneTextureId=PPI_CustomDepth}.Color) {r=true}, Radius=@n27) {attenuationRadius=4}
  n29 = Add(A=@n24, B=Multiply(A=@n25, B=Constant2Vector() {g=-1, r=1}))
  n30 = SphereMask(A=@n22.R, B=ComponentMask(None=SceneTexture(UVs=@n29) {bFiltered=true, sceneTextureId=PPI_CustomDepth}.Color) {r=true}, Radius=@n27) {attenuationRadius=4}
  n31 = Add(A=@n24, B=Multiply(A=@n25, B=Constant2Vector() {g=1, r=-1}))
  n32 = SphereMask(A=@n22.R, B=ComponentMask(None=SceneTexture(UVs=@n31) {bFiltered=true, sceneTextureId=PPI_CustomDepth}.Color) {r=true}, Radius=@n27) {attenuationRadius=4}
  n33 = Max(A=Max(A=@n28, B=@n30), B=@n32)
  n34 = Add(A=@n24, B=Multiply(A=@n25, B=Constant2Vector() {g=-1, r=-1}))
  n35 = SphereMask(A=@n22.R, B=ComponentMask(None=SceneTexture(UVs=@n34) {bFiltered=true, sceneTextureId=PPI_CustomDepth}.Color) {r=true}, Radius=@n27) {attenuationRadius=4}
  n36 = Multiply(A=Multiply(A=@n28, B=@n30), B=@n32)
  MaterialFunctionCall!(Opacity=Subtract(A=Max(A=@n33, B=@n35), B=Multiply(A=@n36, B=@n35))) {materialFunction=/Game/Global/Materials/MaterialGym/Materials/SortedOutline/MF_SortTranslucency}
  n37 = Saturate(None=InverseLinearInterpolate(Value=MaterialFunctionCall() {materialFunction=/Engine/Functions/Engine_MaterialFunctions02/Utility/FOV}.FOV angle in degrees) {constA=45, constB=90}) {desc=FOV compensation}
  Multiply(A=@n10, B=@n37)
  Multiply()
  Multiply(A=@n9, B=@n37)
  n38 = ScalarParameter() {defaultValue=200, parameterName=QuantizedScale}
  n39 = Floor(None=Multiply(A=TextureCoordinate(), B=@n38))
  n40 = TextureSampleParameter2D(UVs=Divide(A=Add(A=@n39) {constB=0.5}, B=@n38)) {parameterName=Diffuse, texture=/BRCosmetics/Characters/Player/Female/Medium/Bodies/F_MED_PhoenixKick_Gain/Textures/T_PhoenixKick_Gain_Body_D}
  n41 = ScalarParameter() {defaultValue=1.25, parameterName=Color Intensity}
  n42 = Floor(None=Multiply(A=Multiply(A=@n40.RGB, B=@n41)) {constB=32})
  n43 = LinearInterpolate(Alpha=ScalarParameter() {parameterName=IsFrontend}) {constA=2}
  n44 = SkyAtmosphereLightDirection()
  n45 = MaterialFunctionCall(Float3=@n44) {materialFunction=&BreakOutFloat3Components}
  n46 = ScalarParameter() {
   desc="0.0 = the light vector is locked on the XY plane.\r\n1.0 = the actual light vector is used  \r\n(for Faces use a blend value of 0,0)"
   group=Pixelation - Shading
   parameterName=Pixelation.LockedZLightVector_Blend
   sliderMax=1
   sortPriority=33
  }
  n47 = Normalize(VectorInput=MaterialFunctionCall(X=@n45.R, Y=@n45.R, Z=Multiply(A=@n45.B, B=@n46)) {materialFunction=&MakeFloat3}.Result)
  n48 = ScalarParameter() {
   controlType=Enumeration
   enumeration=/Game/Global/Materials/Enums/E_Mat_OnOff
   group=Pixelation - Shading
   parameterName=Pixelation.UseLockedZLightVector
   sortPriority=30
  }
  n49 = Multiply(A=ChannelMaskParameter(None=@n18.RGB) {maskChannel=Green, parameterName=LockedZLight Channel, defaultValue=<rgba 0 1 0 0>}.R, B=@n48)
  n50 = Saturate(None=DotProduct(A=LinearInterpolate(A=@n44, B=@n47, Alpha=@n49), B=PixelNormalWS()))
  n51 = MaterialFunctionCall!(Secondary Line B Size, Secondary Line A Size) {materialFunction=&/MF_Pixelation_Scanlines}
  Multiply(
   A=Multiply(A=Divide(A=@n42) {constB=32}, B=LinearInterpolate(B=@n43, Alpha=@n50) {constA=0.4}) {constB=0.5}
   B=Add(A=@n51.Result, B=Multiply(A=@n51.Secondary Lines) {constB=0.5})
  )
  ScalarParameter() {defaultValue=150, group=Pixelation - Scanlines, parameterName=Pixelation.Scanlines_Amount, sortPriority=-1}
  ScalarParameter() {defaultValue=0.8, group=Pixelation - Scanlines, parameterName=Pixelation.Scanlines_MinValue, sortPriority=-1}
 ]
}
