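# Parent material for the PhoenixKick female skeletal-mesh character that adds a pixelation effect, using material attributes with the shading model set by expression.
# A Set Material Attributes node feeds MF_Pixelation_v1, whose result passes into MF_CharacterParent_2023_PostFX.
# A "Construct Base Inputs" composite supplies that function with a constant, Fresnel, distance blend and pre-skinned position.
# The position uses a shader stage switch, with a vertex interpolator for the pixel stage.
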Material /BRCosmetics/Characters/Player/Female/Medium/Bodies/F_MED_PhoenixKick_Gain/Materials/M_Pixelation_Parent
graph {
 properties {
  bAutomaticallySetUsageInEditor=false
  bUseMaterialAttributes=true
  bUsedWithMeshDeformer=true
  bUsedWithMorphTargets=true
  bUsedWithSkeletalMesh=true
  pixelDepthOffsetMode=PDOM_Legacy
  previewMesh=/BRCosmetics/Accessories/FORT_Backpacks/Backpack_F_MED_PhoenixKick/Meshes/F_MED_PhoenixKick_Pack
  shadingModel=MSM_FromMaterialExpression
  usedShadingModels=Default Lit
 }
 nodes [
  Composite() {subgraphName=Construct Base Inputs}
  n1 = Constant()
  MaterialExpressionReroute_4 = Reroute(None=@n1)
  n2 = PreSkinnedPosition()
  n3 = ShaderStageSwitch(PixelShader=VertexInterpolator(VS=@n2).PS, VertexShader=@n2)
  MaterialExpressionReroute_5 = Reroute(None=@n3)
  n4 = MaterialFunctionCall!(Material) {materialFunction=/Game/Packages/Fortress_SharedMaterials/Base_Material_Functions/CharacterBrightnessandScale_Balance}
  MaterialExpressionReroute_6 = Reroute(None=@n4.Distance Blend)
  n5 = Fresnel() {baseReflectFraction=0, exponent=1}
  MaterialExpressionReroute_7 = Reroute(None=@n5)
  MaterialFunctionCall(
   Material=MaterialFunctionCall!(TechArt Mask, Material=SetMaterialAttributes!(MaterialAttributes)) {materialFunction=&/MF_Pixelation_v1}.Result
   M=@MaterialExpressionReroute_4.M
   Pre Skinned Local Position=@MaterialExpressionReroute_5.Pre Skinned Local Position
   Distance Blend=@MaterialExpressionReroute_6.Distance Blend
   Fresnel=@MaterialExpressionReroute_7.Fresnel
  ) {materialFunction=&MF_CharacterParent_2023_PostFX}
  PinBase!(M=@n1, Pre Skinned Local Position=@n3, Distance Blend=@n4.Distance Blend, Fresnel=@n5) {reroutePins [
    M {expression=&self:MaterialExpressionReroute_4}
    Pre Skinned Local Position {expression=&self:MaterialExpressionReroute_5}
    Distance Blend {expression=&self:MaterialExpressionReroute_6}
    Fresnel {expression=&self:MaterialExpressionReroute_7}
   ]}
  PinBase!()
 ]
}
