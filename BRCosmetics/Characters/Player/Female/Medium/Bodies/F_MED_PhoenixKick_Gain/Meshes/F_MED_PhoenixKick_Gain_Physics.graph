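# Physics asset for the PhoenixKick_Gain medium female player body cosmetic, covering its dangling belt pieces and the lower body and torso around them.
# Boxes form two-segment chains at front left, front right and two spine-belt branches, with limited swing and zero twist.
# Pelvis, spine and thigh capsules have free constraints, as do small connector boxes linking belt branches to spine_01_belt.
# The front-left root joint also carries a second, free constraint duplicating its limited one.
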PhysicsAsset /BRCosmetics/Characters/Player/Female/Medium/Bodies/F_MED_PhoenixKick_Gain/Meshes/F_MED_PhoenixKick_Gain_Physics
bodies {
 belt_corr_frnt_l_1_jnt [{
   shapeType=Box
   extentX=5.9121
   extentY=1.5
   extentZ=9.8537
   shapeName=belt_corr_frnt_l_1_jnt_box
   center=<4.0054 -0.0189 -0.6003>
   rotation=<rot 0.0005 -0.0014 179.9997>
  }]
 belt_corr_frnt_l_2_jnt [{
   shapeType=Box
   extentX=5.2
   extentY=1.2
   extentZ=9.8537
   shapeName=belt_corr_frnt_l_2_jnt_box
   center=<3.9927 -0.1665 -0.4854>
   rotation=<rot 1.2925 0.0581 -179.8292>
  }]
 belt_corr_frnt_l_base_jnt [{shapeType=Capsule, length=0, radius=0, shapeName=belt_corr_frnt_l_base_jnt_capsule}]
 belt_corr_frnt_r_1_jnt [{
   shapeType=Box
   extentX=5.7291
   extentY=1.649
   extentZ=10.9402
   shapeName=belt_corr_frnt_r_1_jnt_box
   center=<-4.1002 -0.7659 0>
   rotation=<rot 0 -0.0001 5>
  }]
 belt_corr_frnt_r_2_jnt [{
   shapeType=Box
   extentX=7.217
   extentY=1.0169
   extentZ=12.5585
   shapeName=belt_corr_frnt_r_2_jnt_box
   center=<-2.412 -0.8433 0.1393>
   rotation=<rot 0.0001 -0.0002 10>
  }]
 belt_corr_frnt_r_base_jnt [{shapeType=Capsule, length=0.5, radius=0.5, shapeName=belt_corr_frnt_r_base_jnt_capsule}]
 pelvis [
  {shapeType=Capsule, length=20.6404, radius=6.7792, shapeName=pelvis_capsule, center=<0 -0.3171 0>}
  {shapeType=Capsule, length=3.3575, radius=8.1864, shapeName=pelvis_capsule}
 ]
 spine_01 [{shapeType=Capsule, length=8.8416, radius=7.1584, shapeName=spine_01_capsule, center=<1.9925 -0.8951 0>}]
 spine_01_belt [{shapeType=Capsule, length=3, radius=3, shapeName=spine_01_belt_capsule, center=<0 -0.6489 0>}]
 spine_01_up_belt [{shapeType=Capsule, length=3, radius=2, shapeName=spine_01_up_belt_capsule, center=<1.7974 0 0>}]
 spine_02_belt [{shapeType=Box, extentX=2.9841, extentY=1.5, extentZ=4.4753, center=<1.6223 -0.5675 0.031>, rotation=<rot -0.0001 -0.0001 -9.9999>}]
 spine_02_kn_belt [{shapeType=Box, extentX=1.01, extentY=1.01, extentZ=1.01, shapeName=spine_02_kn_belt_box, rotation=<rot 0 -90 0>}]
 spine_03 [{shapeType=Capsule, length=2.5597, radius=9.3403, shapeName=spine_03_capsule, center=<0 1.2321 0>}]
 spine_03_belt [{shapeType=Box, extentX=5.5727, extentY=1.7237, extentZ=5.5259, center=<2.9847 0.0082 -0.5057>, rotation=<rot -0.0001 10.6249 0.0001>}]
 spine_5_belt [{shapeType=Box, extentX=3.0123, extentY=0.78, extentZ=3.961, shapeName=spine_5_belt_box, center=<1.461 -1 0.393>}]
 spine_5_kn_belt [{shapeType=Box, extentX=1.01, extentY=1.01, extentZ=1.01, shapeName=spine_5_kn_belt_box, rotation=<rot 0 -90 0>}]
 spine_6_belt [{shapeType=Box, extentX=6.6379, extentY=1.6631, extentZ=5.3082, shapeName=spine_6_belt_box, center=<3.1738 -0.4992 -0.3521>}]
 thigh_l [{shapeType=Capsule, length=15.2, radius=7.5, shapeName=thigh_l_capsule, center=<-13.1245 -0.5678 0>, rotation=<rot 90 177.1871 0>}]
 thigh_r [{shapeType=Capsule, length=15.2, radius=7.5, shapeName=thigh_r_capsule, center=<13.1491 1.0686 0>, rotation=<rot 90 -2.8129 0>}]
}
constraints [
 {bone1Name=belt_corr_frnt_l_1_jnt, bone2Name=belt_corr_frnt_l_base_jnt, swing1LimitDegrees=25, swing2LimitDegrees=10, twistLimitDegrees=0}
 {bone1Name=belt_corr_frnt_l_2_jnt, bone2Name=belt_corr_frnt_l_1_jnt, swing1LimitDegrees=60, swing2LimitDegrees=10, twistLimitDegrees=0}
 {bone1Name=spine_01_up_belt, bone2Name=spine_01_belt, swing1LimitDegrees=30, swing2LimitDegrees=30, twistLimitDegrees=0}
 {
  bone1Name=spine_03
  bone2Name=spine_01
  swing1LimitDegrees=45
  swing1Motion=Free
  swing2LimitDegrees=45
  swing2Motion=Free
  twistLimitDegrees=45
  twistMotion=Free
 }
 {bone1Name=belt_corr_frnt_r_1_jnt, bone2Name=belt_corr_frnt_r_base_jnt, swing1LimitDegrees=20, swing2LimitDegrees=10, twistLimitDegrees=0}
 {bone1Name=belt_corr_frnt_r_2_jnt, bone2Name=belt_corr_frnt_r_1_jnt, swing1LimitDegrees=20, swing2LimitDegrees=10, twistLimitDegrees=0}
 {
  bone1Name=spine_01
  bone2Name=pelvis
  swing1LimitDegrees=45
  swing1Motion=Free
  swing2LimitDegrees=45
  swing2Motion=Free
  twistLimitDegrees=45
  twistMotion=Free
 }
 {
  bone1Name=thigh_l
  bone2Name=pelvis
  swing1LimitDegrees=45
  swing1Motion=Free
  swing2LimitDegrees=45
  swing2Motion=Free
  twistLimitDegrees=45
  twistMotion=Free
 }
 {
  bone1Name=thigh_r
  bone2Name=pelvis
  swing1LimitDegrees=45
  swing1Motion=Free
  swing2LimitDegrees=45
  swing2Motion=Free
  twistLimitDegrees=45
  twistMotion=Free
 }
 {
  bone1Name=spine_5_kn_belt
  bone2Name=spine_01_belt
  swing1LimitDegrees=45
  swing1Motion=Free
  swing2LimitDegrees=45
  swing2Motion=Free
  twistLimitDegrees=45
  twistMotion=Free
 }
 {bone1Name=spine_5_belt, bone2Name=spine_5_kn_belt, swing1LimitDegrees=40, swing2LimitDegrees=20, twistLimitDegrees=0}
 {bone1Name=spine_6_belt, bone2Name=spine_5_belt, swing1LimitDegrees=40, swing2LimitDegrees=20, twistLimitDegrees=0}
 {
  bone1Name=spine_02_kn_belt
  bone2Name=spine_01_belt
  swing1LimitDegrees=45
  swing1Motion=Free
  swing2LimitDegrees=45
  swing2Motion=Free
  twistLimitDegrees=45
  twistMotion=Free
 }
 {bone1Name=spine_02_belt, bone2Name=spine_02_kn_belt, swing1LimitDegrees=40, swing2LimitDegrees=20, twistLimitDegrees=0}
 {bone1Name=spine_03_belt, bone2Name=spine_02_belt, swing1LimitDegrees=40, swing2LimitDegrees=20, twistLimitDegrees=0}
 {
  bone1Name=belt_corr_frnt_l_1_jnt
  bone2Name=belt_corr_frnt_l_base_jnt
  swing1LimitDegrees=45
  swing1Motion=Free
  swing2LimitDegrees=45
  swing2Motion=Free
  twistLimitDegrees=45
  twistMotion=Free
 }
]
