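# Animation Blueprint for the F_MED_PhoenixKick_Gain cosmetic body, adding secondary motion to thigh, shorts and belt bones.
# The parent mesh's pose runs through a Control Rig, pelvis-space RigidBody and weighted belt constraint into cached pose PreDyn.
# Layered angular and partial-alpha linear AnimDynamics passes drive thigh bones, with bPbBlendEnabled switching between that result and PreDyn.
# Root-relative AnimDynamics then always simulates shorts, upper-arm belts and back belt bones.
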AnimBlueprint /BRCosmetics/Characters/Player/Female/Medium/Bodies/F_MED_PhoenixKick_Gain/Meshes/F_MED_PhoenixKick_Gain_AnimBP
target_skeleton=&/F_MED_PhoenixKick_Gain_Skeleton
graphs [
 AnimGraph {nodes [
   n1 = Get Sim Space Settings() -> (Sim Space Settings:Anim Phys Sim Space Settings Structure)
   n2 = Anim Dynamics - Bone: thigh_dyn_l_frt!(
    ComponentPose=Local To Component(LocalPose=Use cached pose 'PreDyn'())
    bAlphaBoolEnabled:bool=True
    AngularSpringConstant:float=5000
    SimSpaceSettings:Anim Phys Sim Space Settings Structure=@n1
   ) -> (Pose) {
    alphaInputType=Bool
    angularBiasOverride=1
    angularDampingOverride=0.9999
    angularSpringConstant=5000
    bAngularSpring=true
    bOverrideAngularBias=true
    bOverrideAngularDamping=true
    bUseGravityOverride=true
    bUsePlanarLimit=false
    linearDampingOverride=0.999
    linearSpringConstant=100
    boundBone {boneName=thigh_dyn_l_frt}
    componentAppliedLinearAccClamp=<2 2 2>
    gravityOverride {z=-980}
    physicsBodyDefinitions [{
      boundBone {boneName=thigh_dyn_l_frt}
      boxExtents=<20 20 20>
      constraintSetup {angularLimitsMax=<5 0 0>, angularLimitsMin=<-5 0 0>, angularTarget=<1 0 0>}
      localJointOffset=<-2 0 0>
     }]
    relativeSpaceBone {boneName=spine_04}
    simSpaceSettings {maxAngularAcceleration=45, maxAngularVelocity=8, simSpaceAngularAlpha=0.25}
   }
   n3 = Anim Dynamics - Bone: thigh_dyn_l_bck!(
    ComponentPose=@n2
    bAlphaBoolEnabled:bool=True
    AngularSpringConstant:float=5000
    SimSpaceSettings:Anim Phys Sim Space Settings Structure=@n1
   ) -> (Pose) {
    alphaInputType=Bool
    angularBiasOverride=1
    angularDampingOverride=0.9999
    angularSpringConstant=5000
    bAngularSpring=true
    bOverrideAngularBias=true
    bOverrideAngularDamping=true
    bUseGravityOverride=true
    bUsePlanarLimit=false
    linearDampingOverride=0.999
    linearSpringConstant=100
    boundBone {boneName=thigh_dyn_l_bck}
    componentAppliedLinearAccClamp=<1000 1000 1000>
    componentLinearAccScale=<0.8 0.8 0.8>
    componentLinearVelScale=<0.8 0.8 0.8>
    gravityOverride {z=-980}
    physicsBodyDefinitions [{
      boundBone {boneName=thigh_dyn_l_bck}
      boxExtents=<20 20 20>
      constraintSetup {angularLimitsMax=<10 2 0>, angularLimitsMin=<-10 -2 -2>, angularTarget=<1 0 0>}
      localJointOffset=<-2 0 0>
     }]
    relativeSpaceBone {boneName=spine_04}
    simSpaceSettings {maxAngularAcceleration=45, maxAngularVelocity=8, simSpaceAngularAlpha=0.25}
   }
   n4 = Anim Dynamics - Bone: thigh_dyn_r_frt!(
    ComponentPose=@n3
    bAlphaBoolEnabled:bool=True
    AngularSpringConstant:float=5000
    SimSpaceSettings:Anim Phys Sim Space Settings Structure=(SimSpaceAngularAlpha=0,MaxAngularVelocity=10000,MaxAngularAcceleration=10000,ExternalAngularVelocity=(X=0,Y=0,Z=0))
   ) -> (Pose) {
    alphaInputType=Bool
    angularBiasOverride=1
    angularDampingOverride=0.9999
    angularSpringConstant=5000
    bAngularSpring=true
    bOverrideAngularBias=true
    bOverrideAngularDamping=true
    bUseGravityOverride=true
    bUsePlanarLimit=false
    linearDampingOverride=0.999
    linearSpringConstant=100
    boundBone {boneName=thigh_dyn_r_frt}
    componentAppliedLinearAccClamp=<2 2 2>
    gravityOverride {z=-980}
    physicsBodyDefinitions [{
      boundBone {boneName=thigh_dyn_r_frt}
      boxExtents=<20 20 20>
      constraintSetup {angularLimitsMax=<5 0 0>, angularLimitsMin=<-5 0 0>, angularTarget=<1 0 0>}
      localJointOffset=<-2 0 0>
     }]
    relativeSpaceBone {boneName=spine_04}
    simSpaceSettings {maxAngularAcceleration=45, maxAngularVelocity=8, simSpaceAngularAlpha=0.25}
   }
   n5 = Anim Dynamics - Bone: thigh_dyn_r_bck!(
    ComponentPose=@n4
    bAlphaBoolEnabled:bool=True
    AngularSpringConstant:float=5000
    SimSpaceSettings:Anim Phys Sim Space Settings Structure=(SimSpaceAngularAlpha=0,MaxAngularVelocity=10000,MaxAngularAcceleration=10000,ExternalAngularVelocity=(X=0,Y=0,Z=0))
   ) -> (Pose) {
    alphaInputType=Bool
    angularBiasOverride=1
    angularDampingOverride=0.9999
    angularSpringConstant=5000
    bAngularSpring=true
    bOverrideAngularBias=true
    bOverrideAngularDamping=true
    bUseGravityOverride=true
    bUsePlanarLimit=false
    linearDampingOverride=0.999
    linearSpringConstant=100
    boundBone {boneName=thigh_dyn_r_bck}
    componentAppliedLinearAccClamp=<1000 1000 1000>
    componentLinearAccScale=<0.8 0.8 0.8>
    componentLinearVelScale=<0.8 0.8 0.8>
    gravityOverride {z=-980}
    physicsBodyDefinitions [{
      boundBone {boneName=thigh_dyn_r_bck}
      boxExtents=<20 20 20>
      constraintSetup {angularLimitsMax=<10 2 0>, angularLimitsMin=<-10 -2 -2>, angularTarget=<1 0 0>}
      localJointOffset=<-2 0 0>
     }]
    relativeSpaceBone {boneName=spine_04}
    simSpaceSettings {maxAngularAcceleration=45, maxAngularVelocity=8, simSpaceAngularAlpha=0.25}
   }
   n6 = Anim Dynamics - Bone: thigh_dyn_l_frt!(ComponentPose=@n5, Alpha:float=0.3, GravityOverride:Vector="0,0,-350", LinearSpringConstant:float=180) -> (Pose) {
    angularBiasOverride=1
    angularDampingOverride=0.999
    angularSpringConstant=5000
    bAngularSpring=true
    bGravityOverrideInSimSpace=true
    bLinearSpring=true
    bOverrideAngularBias=true
    bOverrideLinearDamping=true
    bUseGravityOverride=true
    bUsePlanarLimit=false
    linearDampingOverride=1
    linearSpringConstant=180
    boundBone {boneName=thigh_dyn_l_frt}
    componentAppliedLinearAccClamp=<5 5 5>
    gravityOverride {z=-700}
    physicsBodyDefinitions [{
      boundBone {boneName=thigh_dyn_l_frt}
      boxExtents=<100 100 100>
      constraintSetup {angularLimitsMax=<0 0 0>, angularLimitsMin=<0 0 0>, angularTarget=<1 0 0>, linearAxesMax=<1.5 0.5 0>, linearAxesMin=<-1 -0.5 0>}
      localJointOffset=<2 0 0>
     }]
    relativeSpaceBone {boneName=spine_04}
    simSpaceSettings {maxAngularAcceleration=50, maxAngularVelocity=10, simSpaceAngularAlpha=0.1}
   }
   n7 = Anim Dynamics - Bone: thigh_dyn_l_bck!(ComponentPose=@n6, Alpha:float=0.4, GravityOverride:Vector="0,0,-350", LinearSpringConstant:float=210) -> (Pose) {
    angularBiasOverride=1
    angularDampingOverride=0.999
    angularSpringConstant=5000
    bAngularSpring=true
    bGravityOverrideInSimSpace=true
    bLinearSpring=true
    bOverrideAngularBias=true
    bOverrideLinearDamping=true
    bUseGravityOverride=true
    bUsePlanarLimit=false
    linearDampingOverride=1
    linearSpringConstant=180
    boundBone {boneName=thigh_dyn_l_bck}
    componentAppliedLinearAccClamp=<5 5 5>
    gravityOverride {z=-700}
    physicsBodyDefinitions [{
      boundBone {boneName=thigh_dyn_l_bck}
      boxExtents=<100 100 100>
      constraintSetup {angularLimitsMax=<0 0 0>, angularLimitsMin=<0 0 0>, angularTarget=<1 0 0>, linearAxesMax=<2.5 0.5 0>, linearAxesMin=<-2 -0.5 0>}
      localJointOffset=<-2 0 0>
     }]
    relativeSpaceBone {boneName=spine_04}
    simSpaceSettings {maxAngularAcceleration=50, maxAngularVelocity=10, simSpaceAngularAlpha=0.1}
   }
   n8 = Anim Dynamics - Bone: thigh_dyn_r_frt!(ComponentPose=@n7, Alpha:float=0.3, GravityOverride:Vector="0,0,-350", LinearSpringConstant:float=180) -> (Pose) {
    angularBiasOverride=1
    angularDampingOverride=0.999
    angularSpringConstant=5000
    bAngularSpring=true
    bGravityOverrideInSimSpace=true
    bLinearSpring=true
    bOverrideAngularBias=true
    bOverrideLinearDamping=true
    bUseGravityOverride=true
    bUsePlanarLimit=false
    linearDampingOverride=1
    linearSpringConstant=180
    boundBone {boneName=thigh_dyn_r_frt}
    componentAppliedLinearAccClamp=<5 5 5>
    gravityOverride {z=-700}
    physicsBodyDefinitions [{
      boundBone {boneName=thigh_dyn_r_frt}
      boxExtents=<100 100 100>
      constraintSetup {angularLimitsMax=<0 0 0>, angularLimitsMin=<0 0 0>, angularTarget=<1 0 0>, linearAxesMax=<1.5 0.5 0>, linearAxesMin=<-1 -0.5 0>}
      localJointOffset=<2 0 0>
     }]
    relativeSpaceBone {boneName=spine_04}
    simSpaceSettings {maxAngularAcceleration=50, maxAngularVelocity=10, simSpaceAngularAlpha=0.1}
   }
   n9 = Anim Dynamics - Bone: thigh_dyn_r_bck!(ComponentPose=@n8, Alpha:float=0.4, GravityOverride:Vector="0,0,-350", LinearSpringConstant:float=210) -> (Pose) {
    angularBiasOverride=1
    angularDampingOverride=0.999
    angularSpringConstant=5000
    bAngularSpring=true
    bGravityOverrideInSimSpace=true
    bLinearSpring=true
    bOverrideAngularBias=true
    bOverrideLinearDamping=true
    bUseGravityOverride=true
    bUsePlanarLimit=false
    linearDampingOverride=1
    linearSpringConstant=180
    boundBone {boneName=thigh_dyn_r_bck}
    componentAppliedLinearAccClamp=<5 5 5>
    gravityOverride {z=-700}
    physicsBodyDefinitions [{
      boundBone {boneName=thigh_dyn_r_bck}
      boxExtents=<100 100 100>
      constraintSetup {angularLimitsMax=<0 0 0>, angularLimitsMin=<0 0 0>, angularTarget=<1 0 0>, linearAxesMax=<2.5 0.5 0>, linearAxesMin=<-2 -0.5 0>}
      localJointOffset=<-2 0 0>
     }]
    relativeSpaceBone {boneName=spine_04}
    simSpaceSettings {maxAngularAcceleration=50, maxAngularVelocity=10, simSpaceAngularAlpha=0.1}
   }
   n10 = Component To Local(ComponentPose=@n9)
   n11 = Get bPbBlendEnabled() -> (bPbBlendEnabled:bool)
   n12 = Use cached pose 'PreDyn'()
   n13 = Anim Dynamics - Bone: shorts_1_frnt_r_jnt(
    ComponentPose=Local To Component(LocalPose=Blend Poses by bool(BlendPose_0=@n10, BlendPose_1=@n12, bActiveValue=@n11))
    Alpha=1
    ExternalForce="0,0,0"
   ) {
    angularDampingOverride=1
    bAngularSpring=true
    bOverrideAngularDamping=true
    bUseGravityOverride=true
    simulationSpace=RootRelative
    boundBone {boneName=shorts_1_frnt_r_jnt}
    componentAppliedLinearAccClamp=<10000 10000 10000>
    componentLinearAccScale=<0.5 0.5 0.5>
    componentLinearVelScale=<0.5 0.5 0.5>
    gravityOverride {z=-980}
    physicsBodyDefinitions [{
      boundBone {boneName=shorts_1_frnt_r_jnt}
      boxExtents=<10 10 10>
      constraintSetup {angularLimitsMax=<0 5 10>, angularLimitsMin=<0 -5 -10>, angularTarget=<1 0 0>}
      localJointOffset=<-2 0 0>
     }]
    simSpaceSettings {maxAngularAcceleration=1000, maxAngularVelocity=5000, simSpaceAngularAlpha=0.1, externalAngularVelocity=<0.5 0.5 0.5>}
   }
   n14 = Anim Dynamics - Bone: shorts_1_bck_r_jnt(ComponentPose=@n13, Alpha=1, ExternalForce="0,0,0") {
    angularDampingOverride=0.9999
    bAngularSpring=true
    bOverrideAngularDamping=true
    bUseGravityOverride=true
    simulationSpace=RootRelative
    boundBone {boneName=shorts_1_bck_r_jnt}
    componentAppliedLinearAccClamp=<10000 10000 10000>
    componentLinearAccScale=<0.5 0.5 0.5>
    componentLinearVelScale=<0.5 0.5 0.5>
    gravityOverride {z=-980}
    physicsBodyDefinitions [{
      boundBone {boneName=shorts_1_bck_r_jnt}
      boxExtents=<10 10 10>
      constraintSetup {angularLimitsMax=<0 0 10>, angularLimitsMin=<0 0 -20>, angularTarget=<1 0 0>}
      localJointOffset=<-2 0 0>
     }]
    simSpaceSettings {maxAngularAcceleration=1000, maxAngularVelocity=5000, simSpaceAngularAlpha=0.1, externalAngularVelocity=<0.5 0.5 0.5>}
   }
   n15 = Anim Dynamics - Bone: shorts_1_frnt_l_jnt(ComponentPose=@n14, Alpha=1, ExternalForce="0,0,0") {
    angularDampingOverride=0.9999
    bAngularSpring=true
    bOverrideAngularDamping=true
    bUseGravityOverride=true
    simulationSpace=RootRelative
    boundBone {boneName=shorts_1_frnt_l_jnt}
    componentAppliedLinearAccClamp=<10000 10000 10000>
    componentLinearAccScale=<0.5 0.5 0.5>
    componentLinearVelScale=<0.5 0.5 0.5>
    gravityOverride {z=-980}
    physicsBodyDefinitions [{
      boundBone {boneName=shorts_1_frnt_l_jnt}
      boxExtents=<10 10 10>
      constraintSetup {angularLimitsMax=<0 0 10>, angularLimitsMin=<0 0 -20>, angularTarget=<1 0 0>}
      localJointOffset=<2 0 0>
     }]
    simSpaceSettings {maxAngularAcceleration=1000, maxAngularVelocity=5000, simSpaceAngularAlpha=0.1, externalAngularVelocity=<0.5 0.5 0.5>}
   }
   n16 = Anim Dynamics - Bone: shorts_1_bck_l_jnt(ComponentPose=@n15, Alpha=1, ExternalForce="0,0,0") {
    angularDampingOverride=0.9999
    bAngularSpring=true
    bOverrideAngularDamping=true
    bUseGravityOverride=true
    simulationSpace=RootRelative
    boundBone {boneName=shorts_1_bck_l_jnt}
    componentAppliedLinearAccClamp=<10000 10000 10000>
    componentLinearAccScale=<0.5 0.5 0.5>
    componentLinearVelScale=<0.5 0.5 0.5>
    gravityOverride {z=-980}
    physicsBodyDefinitions [{
      boundBone {boneName=shorts_1_bck_l_jnt}
      boxExtents=<10 10 10>
      constraintSetup {angularLimitsMax=<0 0 10>, angularLimitsMin=<0 0 -20>, angularTarget=<1 0 0>}
      localJointOffset=<2 0 0>
     }]
    simSpaceSettings {maxAngularAcceleration=1000, maxAngularVelocity=5000, simSpaceAngularAlpha=0.1, externalAngularVelocity=<0.5 0.5 0.5>}
   }
   n17 = Anim Dynamics - Bone: upperarm_r_beltUp_1_jnt(ComponentPose=@n16, Alpha=1, ExternalForce="0,0,0") {
    angularDampingOverride=0.999
    angularSpringConstant=50
    bAngularSpring=true
    bOverrideAngularDamping=true
    bUseGravityOverride=true
    simulationSpace=RootRelative
    boundBone {boneName=upperarm_r_beltUp_1_jnt}
    componentAppliedLinearAccClamp=<10000 10000 10000>
    componentLinearAccScale=<0.5 0.5 0.5>
    componentLinearVelScale=<0.5 0.5 0.5>
    gravityOverride {z=-980}
    physicsBodyDefinitions [{
      boundBone {boneName=upperarm_r_beltUp_1_jnt}
      boxExtents=<10 10 10>
      constraintSetup {angularLimitsMax=<15 15 10>, angularLimitsMin=<-15 -15 -40>, angularTarget=<1 0 0>}
      localJointOffset=<1 0 0>
     }]
    simSpaceSettings {maxAngularAcceleration=25, maxAngularVelocity=5, simSpaceAngularAlpha=1, externalAngularVelocity=<0.5 0.5 0.5>}
   }
   n18 = Anim Dynamics - Bone: upperarm_r_beltUp_2_jnt(ComponentPose=@n17, Alpha=1, ExternalForce="0,0,0") {
    angularDampingOverride=0.999
    angularSpringConstant=50
    bAngularSpring=true
    bOverrideAngularDamping=true
    bUseGravityOverride=true
    simulationSpace=RootRelative
    boundBone {boneName=upperarm_r_beltUp_2_jnt}
    componentAppliedLinearAccClamp=<10000 10000 10000>
    componentLinearAccScale=<0.5 0.5 0.5>
    componentLinearVelScale=<0.5 0.5 0.5>
    gravityOverride {z=-980}
    physicsBodyDefinitions [{
      boundBone {boneName=upperarm_r_beltUp_2_jnt}
      boxExtents=<10 10 10>
      constraintSetup {angularLimitsMax=<15 15 10>, angularLimitsMin=<-15 -15 -40>, angularTarget=<1 0 0>}
      localJointOffset=<1 0 0>
     }]
    simSpaceSettings {maxAngularAcceleration=25, maxAngularVelocity=5, simSpaceAngularAlpha=1, externalAngularVelocity=<0.5 0.5 0.5>}
   }
   n19 = Anim Dynamics - Bone: upperarm_r_beltDwn_1_jnt(ComponentPose=@n18, Alpha=1, ExternalForce="0,0,0") {
    angularDampingOverride=0.999
    angularSpringConstant=50
    bAngularSpring=true
    bOverrideAngularDamping=true
    bUseGravityOverride=true
    simulationSpace=RootRelative
    boundBone {boneName=upperarm_r_beltDwn_1_jnt}
    componentAppliedLinearAccClamp=<10000 10000 10000>
    componentLinearAccScale=<0.5 0.5 0.5>
    componentLinearVelScale=<0.5 0.5 0.5>
    gravityOverride {z=-980}
    physicsBodyDefinitions [{
      boundBone {boneName=upperarm_r_beltDwn_1_jnt}
      boxExtents=<10 10 10>
      constraintSetup {angularLimitsMax=<15 15 5>, angularLimitsMin=<-15 -15 -40>, angularTarget=<1 0 0>}
      localJointOffset=<1 0 0>
     }]
    simSpaceSettings {maxAngularAcceleration=25, maxAngularVelocity=5, simSpaceAngularAlpha=1, externalAngularVelocity=<0.5 0.5 0.5>}
   }
   n20 = Anim Dynamics - Bone: upperarm_r_beltDwn_2_jnt(ComponentPose=@n19, Alpha=1, ExternalForce="0,0,0") {
    angularDampingOverride=0.999
    angularSpringConstant=50
    bAngularSpring=true
    bOverrideAngularDamping=true
    bUseGravityOverride=true
    simulationSpace=RootRelative
    boundBone {boneName=upperarm_r_beltDwn_2_jnt}
    componentAppliedLinearAccClamp=<10000 10000 10000>
    componentLinearAccScale=<0.5 0.5 0.5>
    componentLinearVelScale=<0.5 0.5 0.5>
    gravityOverride {z=-980}
    physicsBodyDefinitions [{
      boundBone {boneName=upperarm_r_beltDwn_2_jnt}
      boxExtents=<10 10 10>
      constraintSetup {angularLimitsMax=<15 15 5>, angularLimitsMin=<-15 -15 -40>, angularTarget=<1 0 0>}
      localJointOffset=<1 0 0>
     }]
    simSpaceSettings {maxAngularAcceleration=25, maxAngularVelocity=5, simSpaceAngularAlpha=1, externalAngularVelocity=<0.5 0.5 0.5>}
   }
   n21 = Anim Dynamics - Bone: upperarm_l_beltUp_1_jnt(ComponentPose=@n20, Alpha=1, ExternalForce="0,0,0") {
    angularDampingOverride=0.999
    angularSpringConstant=50
    bAngularSpring=true
    bOverrideAngularDamping=true
    bUseGravityOverride=true
    simulationSpace=RootRelative
    boundBone {boneName=upperarm_l_beltUp_1_jnt}
    componentAppliedLinearAccClamp=<10000 10000 10000>
    componentLinearAccScale=<0.5 0.5 0.5>
    componentLinearVelScale=<0.5 0.5 0.5>
    gravityOverride {z=-980}
    physicsBodyDefinitions [{
      boundBone {boneName=upperarm_l_beltUp_1_jnt}
      boxExtents=<10 10 10>
      constraintSetup {angularLimitsMax=<15 15 5>, angularLimitsMin=<-15 -15 -40>, angularTarget=<1 0 0>}
      localJointOffset=<1 0 0>
     }]
    simSpaceSettings {maxAngularAcceleration=25, maxAngularVelocity=5, simSpaceAngularAlpha=1, externalAngularVelocity=<0.5 0.5 0.5>}
   }
   n22 = Anim Dynamics - Bone: upperarm_l_beltUp_2_jnt(ComponentPose=@n21, Alpha=1, ExternalForce="0,0,0") {
    angularDampingOverride=0.999
    angularSpringConstant=50
    bAngularSpring=true
    bOverrideAngularDamping=true
    bUseGravityOverride=true
    simulationSpace=RootRelative
    boundBone {boneName=upperarm_l_beltUp_2_jnt}
    componentAppliedLinearAccClamp=<10000 10000 10000>
    componentLinearAccScale=<0.5 0.5 0.5>
    componentLinearVelScale=<0.5 0.5 0.5>
    gravityOverride {z=-980}
    physicsBodyDefinitions [{
      boundBone {boneName=upperarm_l_beltUp_2_jnt}
      boxExtents=<10 10 10>
      constraintSetup {angularLimitsMax=<15 15 5>, angularLimitsMin=<-15 -15 -40>, angularTarget=<1 0 0>}
      localJointOffset=<1 0 0>
     }]
    simSpaceSettings {maxAngularAcceleration=25, maxAngularVelocity=5, simSpaceAngularAlpha=1, externalAngularVelocity=<0.5 0.5 0.5>}
   }
   n23 = Anim Dynamics - Bone: upperarm_l_beltDwn_1_jnt(ComponentPose=@n22, Alpha=1, ExternalForce="0,0,0") {
    angularDampingOverride=0.999
    angularSpringConstant=50
    bAngularSpring=true
    bOverrideAngularDamping=true
    bUseGravityOverride=true
    simulationSpace=RootRelative
    boundBone {boneName=upperarm_l_beltDwn_1_jnt}
    componentAppliedLinearAccClamp=<10000 10000 10000>
    componentLinearAccScale=<0.5 0.5 0.5>
    componentLinearVelScale=<0.5 0.5 0.5>
    gravityOverride {z=-980}
    physicsBodyDefinitions [{
      boundBone {boneName=upperarm_l_beltDwn_1_jnt}
      boxExtents=<10 10 10>
      constraintSetup {angularLimitsMax=<15 15 5>, angularLimitsMin=<-15 -15 -40>, angularTarget=<1 0 0>}
      localJointOffset=<1 0 0>
     }]
    simSpaceSettings {maxAngularAcceleration=25, maxAngularVelocity=5, simSpaceAngularAlpha=1, externalAngularVelocity=<0.5 0.5 0.5>}
   }
   n24 = Anim Dynamics - Bone: upperarm_l_beltDwn_2_jnt(ComponentPose=@n23, Alpha=1, ExternalForce="0,0,0") {
    angularDampingOverride=0.999
    angularSpringConstant=50
    bAngularSpring=true
    bOverrideAngularDamping=true
    bUseGravityOverride=true
    simulationSpace=RootRelative
    boundBone {boneName=upperarm_l_beltDwn_2_jnt}
    componentAppliedLinearAccClamp=<10000 10000 10000>
    componentLinearAccScale=<0.5 0.5 0.5>
    componentLinearVelScale=<0.5 0.5 0.5>
    gravityOverride {z=-980}
    physicsBodyDefinitions [{
      boundBone {boneName=upperarm_l_beltDwn_2_jnt}
      boxExtents=<10 10 10>
      constraintSetup {angularLimitsMax=<15 15 5>, angularLimitsMin=<-15 -15 -40>, angularTarget=<1 0 0>}
      localJointOffset=<1 0 0>
     }]
    simSpaceSettings {maxAngularAcceleration=25, maxAngularVelocity=5, simSpaceAngularAlpha=1, externalAngularVelocity=<0.5 0.5 0.5>}
   }
   n25 = Anim Dynamics - Bone: belt_corr_bck_r_1_jnt(ComponentPose=@n24, Alpha=1, ExternalForce="0,0,0") {
    angularDampingOverride=1
    angularSpringConstant=100
    bAngularSpring=true
    bOverrideAngularDamping=true
    bUseGravityOverride=true
    simulationSpace=RootRelative
    boundBone {boneName=belt_corr_bck_r_1_jnt}
    componentAppliedLinearAccClamp=<5000 5000 5000>
    componentLinearAccScale=<0.5 0.5 0.5>
    componentLinearVelScale=<0.5 0.5 0.5>
    gravityOverride {z=-980}
    physicsBodyDefinitions [{
      boundBone {boneName=belt_corr_bck_r_1_jnt}
      boxExtents=<11 11 11>
      constraintSetup {angularLimitsMax=<0 0 30>, angularLimitsMin=<0 0 0>, angularTarget=<1 0 0>}
      localJointOffset=<-0.1 0 0>
     }]
    simSpaceSettings {maxAngularAcceleration=25, maxAngularVelocity=5, simSpaceAngularAlpha=1}
   }
   n26 = Anim Dynamics - Bone: belt_corr_bck_r_2_jnt(ComponentPose=@n25, Alpha=1, ExternalForce="0,0,0") {
    angularDampingOverride=1
    angularSpringConstant=100
    bAngularSpring=true
    bOverrideAngularDamping=true
    bUseGravityOverride=true
    simulationSpace=RootRelative
    boundBone {boneName=belt_corr_bck_r_2_jnt}
    componentAppliedLinearAccClamp=<5000 5000 5000>
    componentLinearAccScale=<0.5 0.5 0.5>
    componentLinearVelScale=<0.5 0.5 0.5>
    gravityOverride {z=-980}
    physicsBodyDefinitions [{
      boundBone {boneName=belt_corr_bck_r_2_jnt}
      boxExtents=<12 12 12>
      constraintSetup {angularLimitsMax=<0 5 20>, angularLimitsMin=<0 -5 0>, angularTarget=<-1 0 0>}
      localJointOffset=<-1 0 0>
     }]
    simSpaceSettings {maxAngularAcceleration=25, maxAngularVelocity=5, simSpaceAngularAlpha=1}
   }
   n27 = Anim Dynamics - Bone: belt_corr_bck_l_1_jnt(ComponentPose=@n26, Alpha=1, ExternalForce="0,0,0") {
    angularDampingOverride=1
    angularSpringConstant=100
    bAngularSpring=true
    bOverrideAngularDamping=true
    bUseGravityOverride=true
    simulationSpace=RootRelative
    boundBone {boneName=belt_corr_bck_l_1_jnt}
    componentAppliedLinearAccClamp=<5000 5000 5000>
    componentLinearAccScale=<0.5 0.5 0.5>
    componentLinearVelScale=<0.5 0.5 0.5>
    gravityOverride {z=-980}
    physicsBodyDefinitions [{
      boundBone {boneName=belt_corr_bck_l_1_jnt}
      boxExtents=<11 11 11>
      constraintSetup {angularLimitsMax=<0 0 30>, angularLimitsMin=<0 0 -5>, angularTarget=<1 0 0>}
      localJointOffset=<1 0 0>
     }]
    simSpaceSettings {maxAngularAcceleration=25, maxAngularVelocity=5, simSpaceAngularAlpha=1}
   }
   n28 = Anim Dynamics - Bone: belt_corr_bck_l_2_jnt(ComponentPose=@n27, Alpha=1, ExternalForce="0,0,0") {
    angularDampingOverride=1
    angularSpringConstant=100
    bAngularSpring=true
    bOverrideAngularDamping=true
    bUseGravityOverride=true
    simulationSpace=RootRelative
    boundBone {boneName=belt_corr_bck_l_2_jnt}
    componentAppliedLinearAccClamp=<5000 5000 5000>
    componentLinearAccScale=<0.5 0.5 0.5>
    componentLinearVelScale=<0.5 0.5 0.5>
    gravityOverride {z=-980}
    physicsBodyDefinitions [{
      boundBone {boneName=belt_corr_bck_l_2_jnt}
      boxExtents=<12 12 12>
      constraintSetup {angularLimitsMax=<0 5 30>, angularLimitsMin=<0 -5 0>, angularTarget=<1 0 0>}
      localJointOffset=<3 0 0>
     }]
    simSpaceSettings {maxAngularAcceleration=25, maxAngularVelocity=5, simSpaceAngularAlpha=1}
   }
   n29 = Anim Dynamics - Bone: belt_corr_bck_m_1_jnt(ComponentPose=@n28, Alpha=1, ExternalForce="0,0,0") {
    angularDampingOverride=1
    angularSpringConstant=100
    bAngularSpring=true
    bOverrideAngularDamping=true
    bUseGravityOverride=true
    simulationSpace=RootRelative
    boundBone {boneName=belt_corr_bck_m_1_jnt}
    componentAppliedLinearAccClamp=<5000 5000 5000>
    componentLinearAccScale=<0.5 0.5 0.5>
    componentLinearVelScale=<0.5 0.5 0.5>
    gravityOverride {z=-980}
    physicsBodyDefinitions [{
      boundBone {boneName=belt_corr_bck_m_1_jnt}
      boxExtents=<12 12 12>
      constraintSetup {angularLimitsMax=<10 5 30>, angularLimitsMin=<-10 -5 -5>, angularTarget=<1 0 0>}
      localJointOffset=<1 0 0>
     }]
    simSpaceSettings {maxAngularAcceleration=25, maxAngularVelocity=5, simSpaceAngularAlpha=1}
   }
   n30 = Anim Dynamics - Bone: belt_corr_bck_m_2_jnt(ComponentPose=@n29, Alpha=1, ExternalForce="0,0,0") {
    angularDampingOverride=1
    angularSpringConstant=100
    bAngularSpring=true
    bOverrideAngularDamping=true
    bUseGravityOverride=true
    simulationSpace=RootRelative
    boundBone {boneName=belt_corr_bck_m_2_jnt}
    componentAppliedLinearAccClamp=<5000 5000 5000>
    componentLinearAccScale=<0.5 0.5 0.5>
    componentLinearVelScale=<0.5 0.5 0.5>
    gravityOverride {z=-980}
    physicsBodyDefinitions [{
      boundBone {boneName=belt_corr_bck_m_2_jnt}
      boxExtents=<12 12 12>
      constraintSetup {angularLimitsMax=<0 5 30>, angularLimitsMin=<0 -5 0>, angularTarget=<1 0 0>}
      localJointOffset=<3 0 0>
     }]
    simSpaceSettings {maxAngularAcceleration=25, maxAngularVelocity=5, simSpaceAngularAlpha=1}
   }
   Output Pose(Result=Component To Local(ComponentPose=@n30)) {layerGroup=None}
   n31 = Control Rig!(Source=Copy Pose From Mesh() {bUseAttachedParent=true}, Alpha, CurrentLOD:int=0) -> (Pose) {controlRigAssetReference {controlRigAsset=&/F_MED_PhoenixKick_Gain_CtrlRig}}
   n32 = RigidBody(ComponentPose=Local To Component(LocalPose=@n31)) {
    bClampLinearTranslationLimitToRefPose=true
    bOverrideWorldGravity=true
    overridePhysicsAsset=&/F_MED_PhoenixKick_Gain_Physics
    simulationSpace=BaseBoneSpace
    baseBoneRef {boneName=pelvis}
    overrideWorldGravity {z=-980}
    simSpaceSettings {dampingAlpha=0, maxAngularAcceleration=8, maxAngularVelocity=2, maxLinearAcceleration=1000, maxLinearVelocity=200, worldAlpha=1}
   }
   n33 = "Constraint - spine_01_belt (PPPPPPP)"!(
    ComponentPose=@n32
    Alpha=1
    ConstraintWeights_0=0.1158
    ConstraintWeights_1:float=0.055
    ConstraintWeights_2:float=0.4154
    ConstraintWeights_3:float=0.0244
    ConstraintWeights_4:float=0.3893
    ConstraintWeights_5:float=0
    ConstraintWeights_6:float=0
   ) -> (Pose) {
    boneToModify {boneName=spine_01_belt}
    constraintSetup [
     {targetBone {boneName=spine_04}}
     {targetBone {boneName=spine_03}}
     {targetBone {boneName=spine_02}}
     {targetBone {boneName=spine_01}}
     {targetBone {boneName=pelvis}}
     {targetBone {boneName=belt_corr_frnt_r_1_jnt}}
     {targetBone {boneName=belt_corr_frnt_r_2_jnt}}
    ]
    constraintWeights [1,1,1,1,1,1,1]
   }
   Save cached pose 'PreDyn'(Pose=Component To Local(ComponentPose=@n33)) {cacheName=PreDyn}
  ]}
 EventGraph {nodes [
   Event BlueprintUpdateAnimation()
   Get CurrentLOD()
  ]}
]
variables [Sim Space Settings {
  type=AnimPhysSimSpaceSettings
  default {maxAngularAcceleration=45, maxAngularVelocity=8, simSpaceAngularAlpha=0.25, externalAngularVelocity=<0 0 0>}
 }]
