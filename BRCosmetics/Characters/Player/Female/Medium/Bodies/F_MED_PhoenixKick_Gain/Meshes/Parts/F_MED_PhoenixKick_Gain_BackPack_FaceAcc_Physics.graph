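# Physics asset for the backpack/face accessory part of the female medium PhoenixKick_Gain cosmetic body, covering two jointed chains attached to the head.
# Capsules on upper spine, clavicles, neck and head form the body, linked by constraints with free swing and twist.
# Left and right chains run from base joints through joints 01–04, with joint 06 branching off 02.
# Chain constraints have narrow swing limits and zero twist below the bases.
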PhysicsAsset /BRCosmetics/Characters/Player/Female/Medium/Bodies/F_MED_PhoenixKick_Gain/Meshes/Parts/F_MED_PhoenixKick_Gain_BackPack_FaceAcc_Physics
bodies {
 clavicle_l [{
   shapeType=Capsule
   length=8
   radius=7
   shapeName=clavicle_l_capsule
   center=<3.5228 -1.9828 -0.6304>
   rotation=<rot -78.75 179.999 -179.9992>
  }]
 clavicle_r [{shapeType=Capsule, length=8, radius=7, shapeName=clavicle_r_capsule, center=<-4.2016 1.9534 0.6128>, rotation=<rot -78.75 179.999 -179.9992>}]
 head [{shapeType=Capsule, length=5.5, radius=6.5, shapeName=head_capsule, center=<1.897 2.3656 -0.3263>, rotation=<rot 90 9.9996 0>}]
 head_l_base_jnt [{shapeType=Capsule, length=1, radius=1, shapeName=head_l_base_jnt_capsule}]
 head_l_jnt_01_jnt [{shapeType=Capsule, length=3, radius=1.5, shapeName=head_l_jnt_01_jnt_capsule, center=<2.2191 0 0>}]
 head_l_jnt_02_jnt [{shapeType=Capsule, length=2, radius=2, shapeName=head_l_jnt_02_jnt_capsule, center=<1.719 0.5688 0>}]
 head_l_jnt_03_jnt [{shapeType=Capsule, length=2, radius=2, shapeName=head_l_jnt_03_jnt_capsule, center=<1.719 0 0>}]
 head_l_jnt_04_jnt [{shapeType=Capsule, length=2, radius=2, shapeName=head_l_jnt_04_jnt_capsule, center=<1.719 0 0>, rotation=<rot 0.0004 0 5.6247>}]
 head_l_jnt_06_jnt [{shapeType=Capsule, length=2, radius=1.5, shapeName=head_l_jnt_06_jnt_capsule, rotation=<rot -0.0001 -0.0001 -89.9996>}]
 head_r_base_jnt [{shapeType=Capsule, length=1, radius=1, shapeName=head_r_base_jnt_capsule}]
 head_r_jnt_01_jnt [{shapeType=Capsule, length=3, radius=1.5, shapeName=head_r_jnt_01_jnt_capsule, center=<2.908 -0.1939 0>, rotation=<rot 0 0 2.8125>}]
 head_r_jnt_02_jnt [{shapeType=Capsule, length=2, radius=2, shapeName=head_r_jnt_02_jnt_capsule, center=<2.908 -0.3865 0>, rotation=<rot 0.0001 0 19.6875>}]
 head_r_jnt_03_jnt [{shapeType=Capsule, length=2, radius=2, shapeName=head_r_jnt_03_jnt_capsule, center=<2.908 0 0>, rotation=<rot 0 0 19.6875>}]
 head_r_jnt_04_jnt [{shapeType=Capsule, length=2.5, radius=1.5, shapeName=head_r_jnt_04_jnt_capsule, center=<3.8556 0 0>}]
 head_r_jnt_06_jnt [{
   shapeType=Capsule
   length=1.4103
   radius=0.6434
   shapeName=head_r_jnt_06_jnt_capsule
   center=<1.1148 0.0146 -0.2797>
   rotation=<rot 0 91.0807 -87.1749>
  }]
 neck_01 [{shapeType=Capsule, length=5, radius=5, shapeName=neck_01_capsule, center=<3.8628 0 0>, rotation=<rot 0 -90 0>}]
 spine_05 [{shapeType=Capsule, length=14, radius=10.5, shapeName=spine_05_capsule, center=<3.3414 0.5408 0>, rotation=<rot 0 -170.0014 0>}]
}
constraints [
 {
  bone1Name=head
  bone2Name=neck_01
  swing1LimitDegrees=45
  swing1Motion=Free
  swing2LimitDegrees=45
  swing2Motion=Free
  twistLimitDegrees=45
  twistMotion=Free
 }
 {
  bone1Name=neck_01
  bone2Name=spine_05
  swing1LimitDegrees=45
  swing1Motion=Free
  swing2LimitDegrees=45
  swing2Motion=Free
  twistLimitDegrees=45
  twistMotion=Free
 }
 {
  bone1Name=clavicle_l
  bone2Name=spine_05
  swing1LimitDegrees=45
  swing1Motion=Free
  swing2LimitDegrees=45
  swing2Motion=Free
  twistLimitDegrees=45
  twistMotion=Free
 }
 {
  bone1Name=clavicle_r
  bone2Name=spine_05
  swing1LimitDegrees=45
  swing1Motion=Free
  swing2LimitDegrees=45
  swing2Motion=Free
  twistLimitDegrees=45
  twistMotion=Free
 }
 {bone1Name=head_l_base_jnt, bone2Name=head, swing1LimitDegrees=30, swing2LimitDegrees=30, twistLimitDegrees=10}
 {bone1Name=head_r_base_jnt, bone2Name=head, swing1LimitDegrees=15, swing2LimitDegrees=15, twistLimitDegrees=10}
 {bone1Name=head_l_jnt_01_jnt, bone2Name=head_l_base_jnt, swing1LimitDegrees=12, swing2LimitDegrees=12, twistLimitDegrees=0}
 {bone1Name=head_l_jnt_02_jnt, bone2Name=head_l_jnt_01_jnt, swing1LimitDegrees=12, swing2LimitDegrees=12, twistLimitDegrees=0}
 {bone1Name=head_l_jnt_03_jnt, bone2Name=head_l_jnt_02_jnt, swing1LimitDegrees=12, swing2LimitDegrees=12, twistLimitDegrees=0}
 {bone1Name=head_l_jnt_04_jnt, bone2Name=head_l_jnt_03_jnt, swing1LimitDegrees=12, swing2LimitDegrees=12, twistLimitDegrees=0}
 {bone1Name=head_l_jnt_06_jnt, bone2Name=head_l_jnt_02_jnt, swing1LimitDegrees=15, swing2LimitDegrees=25, twistLimitDegrees=0}
 {bone1Name=head_r_jnt_01_jnt, bone2Name=head_r_base_jnt, swing1LimitDegrees=12, swing2LimitDegrees=12, twistLimitDegrees=0}
 {bone1Name=head_r_jnt_02_jnt, bone2Name=head_r_jnt_01_jnt, swing1LimitDegrees=12, swing2LimitDegrees=12, twistLimitDegrees=0}
 {bone1Name=head_r_jnt_03_jnt, bone2Name=head_r_jnt_02_jnt, swing1LimitDegrees=12, swing2LimitDegrees=12, twistLimitDegrees=0}
 {bone1Name=head_r_jnt_04_jnt, bone2Name=head_r_jnt_03_jnt, swing1LimitDegrees=12, swing2LimitDegrees=12, twistLimitDegrees=0}
 {bone1Name=head_r_jnt_06_jnt, bone2Name=head_r_jnt_02_jnt, swing1LimitDegrees=15, swing2LimitDegrees=15, twistLimitDegrees=0}
]
