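# Anim Blueprint for the face accessory part of the PhoenixKick Gain female body, driving secondary motion on head and hair bones.
# It copies the parent mesh pose, then blends by backbling state, checked when customization loads.
# With backbling, Control Rig and AnimDynamics head-joint chains feed RigidBody using a backpack physics asset; otherwise RigidBody alone uses the standard asset.
# AnimDynamics on the hair bun and strand finishes both paths.
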AnimBlueprint /BRCosmetics/Characters/Player/Female/Medium/Bodies/F_MED_PhoenixKick_Gain/Meshes/Parts/F_MED_PhoenixKick_Gain_FaceAcc_AnimBP
target_skeleton=&/F_MED_PhoenixKick_Gain_FaceAcc_Skeleton
graphs [
 AnimGraph {nodes [
   "Transform (Modify) Bone - Bone: head_base_jnt_m_base_jnt"(Alpha=1, Translation="0,0,0", Rotation="-5,0,0", Scale="1,1,1") {
    rotationMode=BMM_Replace
    rotationSpace=BCS_BoneSpace
    translationMode=BMM_Additive
    boneToModify {boneName=head_base_jnt_m_base_jnt}
    rotation {pitch=-5}
   }
   n1 = Control Rig(Source=Use cached pose 'SavedPose2'()) {controlRigAssetReference {controlRigAsset=&/F_MED_PhoenixKick_Gain_FaceAcc_CtrlRig}}
   n2 = Anim Dynamics - Bone: head_jnt_m_jnt_01_jnt(ComponentPose=Local To Component(LocalPose=@n1), Alpha=1, ExternalForce="0,0,0") {
    angularDampingOverride=0.999
    angularSpringConstant=500
    bOverrideAngularDamping=true
    bUseGravityOverride=true
    simulationSpace=RootRelative
    boundBone {boneName=head_jnt_m_jnt_01_jnt}
    componentAppliedLinearAccClamp=<5000 5000 5000>
    componentLinearAccScale=<0.3 0.3 0.3>
    componentLinearVelScale=<0.3 0.3 0.3>
    gravityOverride {z=-980}
    physicsBodyDefinitions [{
      boundBone {boneName=head_jnt_m_jnt_01_jnt}
      boxExtents=<10 10 10>
      constraintSetup {angularLimitsMax=<0 0 20>, angularLimitsMin=<0 0 -20>, angularTarget=<1 0 0>}
      localJointOffset=<1.1 0 0>
     }]
    simSpaceSettings {maxAngularAcceleration=1000, maxAngularVelocity=5000, simSpaceAngularAlpha=0.1, externalAngularVelocity=<0.5 0.5 0.5>}
   }
   n3 = Anim Dynamics - Bone: head_belt_l_jnt_02_jnt(ComponentPose=@n2, Alpha=1, ExternalForce="0,0,0") {
    angularDampingOverride=0.999
    angularSpringConstant=500
    bOverrideAngularDamping=true
    bUseGravityOverride=true
    simulationSpace=RootRelative
    boundBone {boneName=head_belt_l_jnt_02_jnt}
    componentAppliedLinearAccClamp=<5000 5000 5000>
    componentLinearAccScale=<0.3 0.3 0.3>
    componentLinearVelScale=<0.3 0.3 0.3>
    gravityOverride {z=-980}
    physicsBodyDefinitions [{
      boundBone {boneName=head_belt_l_jnt_02_jnt}
      boxExtents=<10 10 10>
      constraintSetup {angularLimitsMax=<0 0 20>, angularLimitsMin=<0 0 -20>, angularTarget=<1 0 0>}
      localJointOffset=<1.1 0 0>
     }]
    simSpaceSettings {maxAngularAcceleration=1000, maxAngularVelocity=5000, simSpaceAngularAlpha=0.1, externalAngularVelocity=<0.5 0.5 0.5>}
   }
   n4 = Anim Dynamics - Bone: head_belt_l_jnt_03_jnt(ComponentPose=@n3, Alpha=1, ExternalForce="0,0,0") {
    angularDampingOverride=0.999
    angularSpringConstant=500
    bOverrideAngularDamping=true
    bUseGravityOverride=true
    simulationSpace=RootRelative
    boundBone {boneName=head_belt_l_jnt_03_jnt}
    componentAppliedLinearAccClamp=<5000 5000 5000>
    componentLinearAccScale=<0.3 0.3 0.3>
    componentLinearVelScale=<0.3 0.3 0.3>
    gravityOverride {z=-980}
    physicsBodyDefinitions [{
      boundBone {boneName=head_belt_l_jnt_03_jnt}
      boxExtents=<10 10 10>
      constraintSetup {angularLimitsMax=<0 0 20>, angularLimitsMin=<0 0 -20>, angularTarget=<1 0 0>}
      localJointOffset=<1.1 0 0>
     }]
    simSpaceSettings {maxAngularAcceleration=1000, maxAngularVelocity=5000, simSpaceAngularAlpha=0.1, externalAngularVelocity=<0.5 0.5 0.5>}
   }
   n5 = Anim Dynamics - Bone: head_jnt_r_jnt_02_jnt(ComponentPose=@n4, Alpha=1, ExternalForce="0,0,0") {
    angularDampingOverride=0.999
    angularSpringConstant=500
    bOverrideAngularDamping=true
    bUseGravityOverride=true
    simulationSpace=RootRelative
    boundBone {boneName=head_jnt_r_jnt_02_jnt}
    componentAppliedLinearAccClamp=<5000 5000 5000>
    componentLinearAccScale=<0.3 0.3 0.3>
    componentLinearVelScale=<0.3 0.3 0.3>
    gravityOverride {z=-980}
    physicsBodyDefinitions [{
      boundBone {boneName=head_jnt_r_jnt_02_jnt}
      boxExtents=<10 10 10>
      constraintSetup {angularLimitsMax=<0 0 20>, angularLimitsMin=<0 0 -20>, angularTarget=<1 0 0>}
      localJointOffset=<1.1 0 0>
     }]
    simSpaceSettings {maxAngularAcceleration=1000, maxAngularVelocity=5000, simSpaceAngularAlpha=0.1, externalAngularVelocity=<0.5 0.5 0.5>}
   }
   n6 = Anim Dynamics - Bone: head_jnt_r_jnt_03_jnt(ComponentPose=@n5, Alpha=1, ExternalForce="0,0,0") {
    angularDampingOverride=0.999
    angularSpringConstant=500
    bOverrideAngularDamping=true
    bUseGravityOverride=true
    simulationSpace=RootRelative
    boundBone {boneName=head_jnt_r_jnt_03_jnt}
    componentAppliedLinearAccClamp=<5000 5000 5000>
    componentLinearAccScale=<0.3 0.3 0.3>
    componentLinearVelScale=<0.3 0.3 0.3>
    gravityOverride {z=-980}
    physicsBodyDefinitions [{
      boundBone {boneName=head_jnt_r_jnt_03_jnt}
      boxExtents=<10 10 10>
      constraintSetup {angularLimitsMax=<0 0 20>, angularLimitsMin=<0 0 -20>, angularTarget=<1 0 0>}
      localJointOffset=<1.1 0 0>
     }]
    simSpaceSettings {maxAngularAcceleration=1000, maxAngularVelocity=5000, simSpaceAngularAlpha=0.1, externalAngularVelocity=<0.5 0.5 0.5>}
   }
   n7 = Anim Dynamics - Bone: head_jnt_m_jnt_02_jnt(ComponentPose=@n6, Alpha=1, ExternalForce="0,0,0") {
    angularDampingOverride=0.999
    angularSpringConstant=500
    bOverrideAngularDamping=true
    bUseGravityOverride=true
    simulationSpace=RootRelative
    boundBone {boneName=head_jnt_m_jnt_02_jnt}
    componentAppliedLinearAccClamp=<5000 5000 5000>
    componentLinearAccScale=<0.3 0.3 0.3>
    componentLinearVelScale=<0.3 0.3 0.3>
    gravityOverride {z=-980}
    physicsBodyDefinitions [{
      boundBone {boneName=head_jnt_m_jnt_02_jnt}
      boxExtents=<10 10 10>
      constraintSetup {angularLimitsMax=<0 0 20>, angularLimitsMin=<0 0 -20>, angularTarget=<1 0 0>}
      localJointOffset=<1.1 0 0>
     }]
    simSpaceSettings {maxAngularAcceleration=1000, maxAngularVelocity=5000, simSpaceAngularAlpha=0.1, externalAngularVelocity=<0.5 0.5 0.5>}
   }
   n8 = Anim Dynamics - Bone: head_jnt_m_jnt_03_jnt(ComponentPose=@n7, Alpha=1, ExternalForce="0,0,0") {
    angularDampingOverride=0.999
    angularSpringConstant=500
    bOverrideAngularDamping=true
    bUseGravityOverride=true
    simulationSpace=RootRelative
    boundBone {boneName=head_jnt_m_jnt_03_jnt}
    componentAppliedLinearAccClamp=<5000 5000 5000>
    componentLinearAccScale=<0.3 0.3 0.3>
    componentLinearVelScale=<0.3 0.3 0.3>
    gravityOverride {z=-980}
    physicsBodyDefinitions [{
      boundBone {boneName=head_jnt_m_jnt_03_jnt}
      boxExtents=<10 10 10>
      constraintSetup {angularLimitsMax=<0 0 20>, angularLimitsMin=<0 0 -20>, angularTarget=<1 0 0>}
      localJointOffset=<1.1 0 0>
     }]
    simSpaceSettings {maxAngularAcceleration=1000, maxAngularVelocity=5000, simSpaceAngularAlpha=0.1, externalAngularVelocity=<0.5 0.5 0.5>}
   }
   n9 = RigidBody(ComponentPose=@n8) {
    bClampLinearTranslationLimitToRefPose=true
    bOverrideWorldGravity=true
    overridePhysicsAsset=&/F_MED_PhoenixKick_Gain_BackPack_FaceAcc_Physics
    simulationSpace=BaseBoneSpace
    baseBoneRef {boneName=spine_05}
    overrideWorldGravity {z=-980}
    simSpaceSettings {
     dampingAlpha=0
     maxAngularAcceleration=8
     maxAngularVelocity=2
     maxLinearAcceleration=1200
     maxLinearVelocity=300
     worldAlpha=1
     externalLinearDragV=<0.2 0.2 0.2>
    }
   }
   n10 = Component To Local(ComponentPose=@n9)
   n11 = RigidBody(ComponentPose=Local To Component(LocalPose=Use cached pose 'SavedPose2'())) {
    bClampLinearTranslationLimitToRefPose=true
    bOverrideWorldGravity=true
    overridePhysicsAsset=&/F_MED_PhoenixKick_Gain_FaceAcc_Physics
    simulationSpace=BaseBoneSpace
    baseBoneRef {boneName=spine_05}
    overrideWorldGravity {z=-980}
    simSpaceSettings {
     dampingAlpha=0
     maxAngularAcceleration=8
     maxAngularVelocity=2
     maxLinearAcceleration=1200
     maxLinearVelocity=300
     worldAlpha=1
     externalLinearDragV=<0.2 0.2 0.2>
    }
   }
   n12 = Component To Local(ComponentPose=@n11)
   n13 = Get Is Wearing Backbling()
   n14 = Anim Dynamics - Bone: hair_bun(
    ComponentPose=Local To Component(LocalPose=Blend Poses by bool(BlendPose_0=@n10, BlendPose_1=@n12, bActiveValue=@n13))
    Alpha=1
    ExternalForce="0,0,0"
   ) {
    angularDampingOverride=1
    angularSpringConstant=5000
    bAngularSpring=true
    bOverrideAngularDamping=true
    bUseGravityOverride=true
    simulationSpace=RootRelative
    boundBone {boneName=hair_bun}
    componentAppliedLinearAccClamp=<500 500 500>
    componentLinearAccScale=<0.3 0.3 0.3>
    componentLinearVelScale=<0.3 0.3 0.3>
    gravityOverride {z=-980}
    physicsBodyDefinitions [{
      boundBone {boneName=hair_bun}
      boxExtents=<14 14 14>
      constraintSetup {angularLimitsMax=<10 0 10>, angularLimitsMin=<-10 0 -10>, angularTarget=<1 0 0>}
      localJointOffset=<1 0 0>
     }]
    simSpaceSettings {maxAngularAcceleration=1000, maxAngularVelocity=5000, simSpaceAngularAlpha=0.1, externalAngularVelocity=<0.5 0.5 0.5>}
   }
   n15 = Anim Dynamics - Bone: hair_m_01_jnt(ComponentPose=@n14, Alpha=1, ExternalForce="0,0,0") {
    angularDampingOverride=0.999
    angularSpringConstant=500
    bOverrideAngularDamping=true
    bUseGravityOverride=true
    simulationSpace=RootRelative
    boundBone {boneName=hair_m_01_jnt}
    componentAppliedLinearAccClamp=<5000 5000 5000>
    componentLinearAccScale=<0.3 0.3 0.3>
    componentLinearVelScale=<0.3 0.3 0.3>
    gravityOverride {z=-980}
    physicsBodyDefinitions [{
      boundBone {boneName=hair_m_01_jnt}
      boxExtents=<10 10 10>
      constraintSetup {angularLimitsMax=<5 15 0>, angularLimitsMin=<-5 -15 -5>, angularTarget=<1 0 0>}
      localJointOffset=<1.1 0 0>
     }]
    simSpaceSettings {maxAngularAcceleration=1000, maxAngularVelocity=5000, simSpaceAngularAlpha=0.1, externalAngularVelocity=<0.5 0.5 0.5>}
   }
   n16 = Anim Dynamics - Bone: hair_m_02_jnt(ComponentPose=@n15, Alpha=1, ExternalForce="0,0,0") {
    angularDampingOverride=0.999
    angularSpringConstant=500
    bOverrideAngularDamping=true
    bUseGravityOverride=true
    simulationSpace=RootRelative
    boundBone {boneName=hair_m_02_jnt}
    componentAppliedLinearAccClamp=<5000 5000 5000>
    componentLinearAccScale=<0.3 0.3 0.3>
    componentLinearVelScale=<0.3 0.3 0.3>
    gravityOverride {z=-980}
    physicsBodyDefinitions [{
      boundBone {boneName=hair_m_02_jnt}
      boxExtents=<10 10 10>
      constraintSetup {angularLimitsMax=<5 15 0>, angularLimitsMin=<-5 -15 -10>, angularTarget=<1 0 0>}
      localJointOffset=<1.1 0 0>
     }]
    simSpaceSettings {maxAngularAcceleration=1000, maxAngularVelocity=5000, simSpaceAngularAlpha=0.1, externalAngularVelocity=<0.5 0.5 0.5>}
   }
   Output Pose(Result=Component To Local(ComponentPose=@n16)) {layerGroup=None}
   Save cached pose 'SavedPose2'(Pose=Copy Pose From Mesh() {bUseAttachedParent=true}) {cacheName=SavedPose2}
  ]}
 EventGraph {nodes [
   Event OnCharacterCustomizationLoaded()
   n1 = CheckForBackbling()
   > Set Is Wearing Backbling(Is Wearing Backbling=OR Boolean!(A=@n1, B=false, C:bool=@n1.bIsBackblingLong) -> (ReturnValue))
   Event BlueprintUpdateAnimation()
  ]}
]
variables [Is Wearing Backbling {type=boolean}]
